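# Widget Blueprint for one Athena mini challenge entry, showing progress, description and XP reward, both in-zone and in frontend with different sizes and layout.
# UpdateProgressBar stores counts that drive the circular icon material's Progress and DeltaProgress parameters.
# Animations play fill bursts, completion checkmark, refresh to a new quest, and a completion toast, which fires dispatchers.
# In-zone XP text is abbreviated to K, with boost styling when first-time boosted.
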WidgetBlueprint /FortniteCoreUI/Athena/UI/Challenges/MiniChallengeEntry
animations [
 OnDeltaProgressFilled {
  binding_count=2
  length_seconds=1.38
  object=&self:OnDeltaProgressFilled
  sequence {
   bindings [
    Burst {
     id=#1
     possessed_class=Image
     tracks [
      {
       type=MovieSceneColorTrack
       display_name=Color and Opacity
       sections [{
         type=MovieSceneColorSection
         channels [
          Color.R {type=MovieSceneScriptingFloatChannel, const [0.0331,1]}
          Color.G {type=MovieSceneScriptingFloatChannel, const [0.855,1]}
          Color.B {type=MovieSceneScriptingFloatChannel, const [1,1]}
          Color.A {type=MovieSceneScriptingFloatChannel, const [1,1]}
         ]
        }]
      }
      {
       type=MovieSceneByteTrack
       display_name=Visibility
       sections [{type=MovieSceneByteSection, channels [None {type=MovieSceneScriptingByteChannel, keys [[20,3],[56,1]]}]}]
      }
      {
       type=MovieSceneFloatTrack
       display_name=Render Opacity
       sections [{type=MovieSceneFloatSection, channels [None {type=MovieSceneScriptingFloatChannel, keys [[0,0],[20,0],[30,0.2],[35,0.2],[56,0]]}]}]
      }
      {
       type=MovieScene2DTransformTrack
       sections [{
         type=MovieScene2DTransformSection
         channels [
          Translation.X {type=MovieSceneScriptingFloatChannel}
          Translation.Y {type=MovieSceneScriptingFloatChannel}
          Angle {type=MovieSceneScriptingFloatChannel}
          Scale.X {type=MovieSceneScriptingFloatChannel, keys [[20,1],[56,2]]}
          Scale.Y {type=MovieSceneScriptingFloatChannel, keys [[20,1],[56,2]]}
          Shear.X {type=MovieSceneScriptingFloatChannel}
          Shear.Y {type=MovieSceneScriptingFloatChannel}
         ]
        }]
      }
     ]
    }
    ChallengeProgressBar {
     id=#2
     possessed_class=Image
     tracks [
      {
       type=MovieSceneWidgetMaterialTrack
       sections [{type=MovieSceneParameterSection, channels [AnimateDeltaProgress {type=MovieSceneScriptingFloatChannel, keys [[66,0],[122,1]]}]}]
      }
      {
       type=MovieScene2DTransformTrack
       sections [{
         type=MovieScene2DTransformSection
         channels [
          Translation.X {type=MovieSceneScriptingFloatChannel}
          Translation.Y {type=MovieSceneScriptingFloatChannel}
          Angle {type=MovieSceneScriptingFloatChannel}
          Scale.X {type=MovieSceneScriptingFloatChannel, const [1,2]}
          Scale.Y {type=MovieSceneScriptingFloatChannel, const [1,2]}
          Shear.X {type=MovieSceneScriptingFloatChannel}
          Shear.Y {type=MovieSceneScriptingFloatChannel}
         ]
        }]
      }
     ]
    }
   ]
   display_rate {numerator=100}
  }
 }
 OnDeltaProgressFilledToComplete {
  binding_count=5
  length_seconds=2
  object=&self:OnDeltaProgressFilledToComplete
  sequence {
   bindings [
    ChallengeProgressBar {
     id=#3
     possessed_class=Image
     tracks [
      {
       type=MovieSceneWidgetMaterialTrack
       sections [{
         type=MovieSceneParameterSection
         channels [
          AnimateDeltaProgress {type=MovieSceneScriptingFloatChannel, keys [[66,0],[122,1]]}
          AnimateProgressBarColor {type=MovieSceneScriptingFloatChannel, keys [[130,0],[138,1]]}
          Progress-Alpha {type=MovieSceneScriptingFloatChannel, keys [[130,1],[138,0]]}
         ]
        }]
      }
      {
       type=MovieScene2DTransformTrack
       sections [{
         type=MovieScene2DTransformSection
         channels [
          Translation.X {type=MovieSceneScriptingFloatChannel}
          Translation.Y {type=MovieSceneScriptingFloatChannel}
          Angle {type=MovieSceneScriptingFloatChannel}
          Scale.X {type=MovieSceneScriptingFloatChannel, const [1,2]}
          Scale.Y {type=MovieSceneScriptingFloatChannel, const [1,2]}
          Shear.X {type=MovieSceneScriptingFloatChannel}
          Shear.Y {type=MovieSceneScriptingFloatChannel}
         ]
        }]
      }
     ]
    }
    RewardVB {
     id=#4
     possessed_class=VerticalBox
     tracks [
      {
       type=MovieScene2DTransformTrack
       sections [{
         type=MovieScene2DTransformSection
         channels [
          Translation.X {type=MovieSceneScriptingFloatChannel}
          Translation.Y {type=MovieSceneScriptingFloatChannel, keys [[122,0],[146,-32]]}
          Angle {type=MovieSceneScriptingFloatChannel}
          Scale.X {type=MovieSceneScriptingFloatChannel, keys [[122,1],[136,0.75]]}
          Scale.Y {type=MovieSceneScriptingFloatChannel, keys [[122,1],[136,1.25]]}
          Shear.X {type=MovieSceneScriptingFloatChannel}
          Shear.Y {type=MovieSceneScriptingFloatChannel}
         ]
        }]
      }
      {
       type=MovieSceneFloatTrack
       display_name=Render Opacity
       sections [{type=MovieSceneFloatSection, channels [None {type=MovieSceneScriptingFloatChannel, keys [[122,1],[136,0]]}]}]
      }
     ]
    }
    CompleteCheck {
     id=#5
     possessed_class=Image
     tracks [
      {
       type=MovieSceneByteTrack
       display_name=Visibility
       sections [{type=MovieSceneByteSection, channels [None {type=MovieSceneScriptingByteChannel, const [3,1]}]}]
      }
      {
       type=MovieSceneFloatTrack
       display_name=Render Opacity
       sections [{type=MovieSceneFloatSection, channels [None {type=MovieSceneScriptingFloatChannel, keys [[132,0],[142,1],[200,1]]}]}]
      }
      {
       type=MovieScene2DTransformTrack
       sections [{
         type=MovieScene2DTransformSection
         channels [
          Translation.X {type=MovieSceneScriptingFloatChannel}
          Translation.Y {type=MovieSceneScriptingFloatChannel}
          Angle {type=MovieSceneScriptingFloatChannel}
          Scale.X {type=MovieSceneScriptingFloatChannel, const [1,2]}
          Scale.Y {type=MovieSceneScriptingFloatChannel, const [1,2]}
          Shear.X {type=MovieSceneScriptingFloatChannel}
          Shear.Y {type=MovieSceneScriptingFloatChannel}
         ]
        }]
      }
     ]
    }
    QuestIcon {
     id=#6
     possessed_class=FortLazyImage
     tracks [
      {
       type=MovieScene2DTransformTrack
       sections [{
         type=MovieScene2DTransformSection
         channels [
          Translation.X {type=MovieSceneScriptingFloatChannel}
          Translation.Y {type=MovieSceneScriptingFloatChannel}
          Angle {type=MovieSceneScriptingFloatChannel}
          Scale.X {type=MovieSceneScriptingFloatChannel, keys [[122,1],[132,0.5]]}
          Scale.Y {type=MovieSceneScriptingFloatChannel, keys [[122,1],[132,0.5]]}
          Shear.X {type=MovieSceneScriptingFloatChannel}
          Shear.Y {type=MovieSceneScriptingFloatChannel}
         ]
        }]
      }
      {
       type=MovieSceneFloatTrack
       display_name=Render Opacity
       sections [{type=MovieSceneFloatSection, channels [None {type=MovieSceneScriptingFloatChannel, keys [[122,1],[132,0]]}]}]
      }
     ]
    }
    Burst {
     id=#7
     possessed_class=Image
     tracks [
      {
       type=MovieSceneColorTrack
       display_name=Color and Opacity
       sections [{
         type=MovieSceneColorSection
         channels [
          Color.R {type=MovieSceneScriptingFloatChannel, keys [[20,0.0331],[56,0.0331],[122,0.0343]]}
          Color.G {type=MovieSceneScriptingFloatChannel, keys [[20,0.855],[56,0.855],[122,1]]}
          Color.B {type=MovieSceneScriptingFloatChannel, keys [[20,1],[56,1],[122,0.227]]}
          Color.A {type=MovieSceneScriptingFloatChannel, const [1,1]}
         ]
        }]
      }
      {
       type=MovieSceneByteTrack
       display_name=Visibility
       sections [{type=MovieSceneByteSection, channels [None {type=MovieSceneScriptingByteChannel, keys [[20,3],[158,1]]}]}]
      }
      {
       type=MovieSceneFloatTrack
       display_name=Render Opacity
       sections [{
         type=MovieSceneFloatSection
         channels [None {type=MovieSceneScriptingFloatChannel, keys [[0.0017,0],[20,0],[30,0.2],[35,0.2],[56,0],[122,0],[132,0.2],[137,0.2],[158,0]]}]
        }]
      }
      {
       type=MovieScene2DTransformTrack
       sections [{
         type=MovieScene2DTransformSection
         channels [
          Translation.X {type=MovieSceneScriptingFloatChannel}
          Translation.Y {type=MovieSceneScriptingFloatChannel}
          Angle {type=MovieSceneScriptingFloatChannel}
          Scale.X {type=MovieSceneScriptingFloatChannel, keys [[20,1],[56,2],[122,1],[158,2]]}
          Scale.Y {type=MovieSceneScriptingFloatChannel, keys [[20,1],[56,2],[122,1],[158,2]]}
          Shear.X {type=MovieSceneScriptingFloatChannel}
          Shear.Y {type=MovieSceneScriptingFloatChannel}
         ]
        }]
      }
     ]
    }
   ]
   display_rate {numerator=100}
  }
 }
 OnCompleteRefreshToNew {
  binding_count=6
  length_seconds=1.82
  object=&self:OnCompleteRefreshToNew
  sequence {
   bindings [
    RewardVB {
     id=#8
     possessed_class=VerticalBox
     tracks [
      {
       type=MovieSceneFloatTrack
       display_name=Render Opacity
       sections [{type=MovieSceneFloatSection, channels [None {type=MovieSceneScriptingFloatChannel, keys [[140,0],[153.9983,1]]}]}]
      }
      {
       type=MovieScene2DTransformTrack
       sections [{
         type=MovieScene2DTransformSection
         channels [
          Translation.X {type=MovieSceneScriptingFloatChannel}
          Translation.Y {type=MovieSceneScriptingFloatChannel, keys [[145.9983,-16],[153.9983,0]]}
          Angle {type=MovieSceneScriptingFloatChannel}
          Scale.X {type=MovieSceneScriptingFloatChannel, keys [[140,0.75],[153.9983,1]]}
          Scale.Y {type=MovieSceneScriptingFloatChannel, keys [[140,1.25],[153.9983,1]]}
          Shear.X {type=MovieSceneScriptingFloatChannel}
          Shear.Y {type=MovieSceneScriptingFloatChannel}
         ]
        }]
      }
     ]
    }
    Burst {
     id=#9
     possessed_class=Image
     tracks [
      {
       type=MovieSceneColorTrack
       display_name=Color and Opacity
       sections [{
         type=MovieSceneColorSection
         channels [
          Color.R {type=MovieSceneScriptingFloatChannel, const [0.0331,1]}
          Color.G {type=MovieSceneScriptingFloatChannel, const [0.855,1]}
          Color.B {type=MovieSceneScriptingFloatChannel, const [1,1]}
          Color.A {type=MovieSceneScriptingFloatChannel, const [1,1]}
         ]
        }]
      }
      {
       type=MovieSceneByteTrack
       display_name=Visibility
       sections [{type=MovieSceneByteSection, channels [None {type=MovieSceneScriptingByteChannel, keys [[49.9983,3],[181.9983,1]]}]}]
      }
      {
       type=MovieSceneFloatTrack
       display_name=Render Opacity
       sections [{
         type=MovieSceneFloatSection
         channels [None {type=MovieSceneScriptingFloatChannel, keys [[0,0],[139.9983,0],[149.9983,0.2],[154.9983,0.2],[175.9983,0]]}]
        }]
      }
      {
       type=MovieScene2DTransformTrack
       sections [{
         type=MovieScene2DTransformSection
         channels [
          Translation.X {type=MovieSceneScriptingFloatChannel}
          Translation.Y {type=MovieSceneScriptingFloatChannel}
          Angle {type=MovieSceneScriptingFloatChannel}
          Scale.X {type=MovieSceneScriptingFloatChannel, keys [[139.9983,1],[175.9983,2]]}
          Scale.Y {type=MovieSceneScriptingFloatChannel, keys [[139.9983,1],[175.9983,2]]}
          Shear.X {type=MovieSceneScriptingFloatChannel}
          Shear.Y {type=MovieSceneScriptingFloatChannel}
         ]
        }]
      }
     ]
    }
    ProgressBarSB {
     id=#10
     possessed_class=SizeBox
     tracks [
      {
       type=MovieScene2DTransformTrack
       sections [{
         type=MovieScene2DTransformSection
         channels [
          Translation.X {type=MovieSceneScriptingFloatChannel}
          Translation.Y {type=MovieSceneScriptingFloatChannel}
          Angle {type=MovieSceneScriptingFloatChannel}
          Scale.X {type=MovieSceneScriptingFloatChannel, keys [[49.9983,1],[81.9983,0.5],[129.9983,0.5],[149.9983,1]]}
          Scale.Y {type=MovieSceneScriptingFloatChannel, keys [[49.9983,1],[81.9983,0.5],[129.9983,0.5],[149.9983,1]]}
          Shear.X {type=MovieSceneScriptingFloatChannel}
          Shear.Y {type=MovieSceneScriptingFloatChannel}
         ]
        }]
      }
      {
       type=MovieSceneFloatTrack
       display_name=Render Opacity
       sections [{
         type=MovieSceneFloatSection
         channels [None {type=MovieSceneScriptingFloatChannel, keys [[73.9983,1],[81.9983,0],[129.9983,0],[139.9983,1]]}]
        }]
      }
     ]
    }
    CompleteCheck {
     id=#5
     possessed_class=Image
     tracks [
      {
       type=MovieSceneByteTrack
       display_name=Visibility
       sections [{type=MovieSceneByteSection, channels [None {type=MovieSceneScriptingByteChannel, keys [[49.9983,3],[81.9983,1]]}]}]
      }
      {
       type=MovieSceneFloatTrack
       display_name=Render Opacity
       sections [{type=MovieSceneFloatSection, channels [None {type=MovieSceneScriptingFloatChannel, keys [[49.9983,1],[81.9983,1],[129.9983,0]]}]}]
      }
     ]
    }
    QuestIcon {
     id=#6
     possessed_class=FortLazyImage
     tracks [
      {
       type=MovieSceneFloatTrack
       display_name=Render Opacity
       sections [{type=MovieSceneFloatSection, channels [None {type=MovieSceneScriptingFloatChannel, keys [[139.9983,0],[149.9983,1]]}]}]
      }
      {
       type=MovieScene2DTransformTrack
       sections [{
         type=MovieScene2DTransformSection
         channels [
          Translation.X {type=MovieSceneScriptingFloatChannel}
          Translation.Y {type=MovieSceneScriptingFloatChannel}
          Angle {type=MovieSceneScriptingFloatChannel}
          Scale.X {type=MovieSceneScriptingFloatChannel, keys [[139.9983,0.5],[153.9983,1]]}
          Scale.Y {type=MovieSceneScriptingFloatChannel, keys [[139.9983,0.5],[153.9983,1]]}
          Shear.X {type=MovieSceneScriptingFloatChannel}
          Shear.Y {type=MovieSceneScriptingFloatChannel}
         ]
        }]
      }
     ]
    }
    ChallengeProgressBar {
     id=#2
     possessed_class=Image
     tracks [{
       type=MovieSceneWidgetMaterialTrack
       sections [{
         type=MovieSceneParameterSection
         channels [
          AnimateProgressBarColor {type=MovieSceneScriptingFloatChannel, keys [[49.9983,1],[81.9983,1],[129.9983,0]]}
          Progress-Alpha {type=MovieSceneScriptingFloatChannel, keys [[49.9983,0],[81.9983,0],[129.9983,1]]}
          AnimateDeltaProgress {type=MovieSceneScriptingFloatChannel, keys [[81.9983,1],[129.9983,0]]}
         ]
        }]
      }]
    }
   ]
   display_rate {numerator=100}
  }
 }
 TransientXPBoostedBurst {
  binding_count=6
  length_seconds=7
  object=&self:TransientXPBoostedBurst
  sequence {
   bindings [
    BoostTrianglesList {
     id=#11
     possessed_class=Image
     tracks [
      {
       type=MovieSceneFloatTrack
       display_name=Render Opacity
       sections [{type=MovieSceneFloatSection, channels [None {type=MovieSceneScriptingFloatChannel, keys [[25,0],[65,0.45],[400,0]]}]}]
      }
      {
       type=MovieSceneByteTrack
       display_name=Visibility
       sections [{type=MovieSceneByteSection, channels [None {type=MovieSceneScriptingByteChannel, keys [[25,3],[400,1]]}]}]
      }
     ]
    }
    BoostGlowList {
     id=#12
     possessed_class=Image
     tracks [
      {
       type=MovieSceneFloatTrack
       display_name=Render Opacity
       sections [{type=MovieSceneFloatSection, channels [None {type=MovieSceneScriptingFloatChannel, keys [[0,0],[25,0],[65,0.45],[400,0]]}]}]
      }
      {
       type=MovieSceneByteTrack
       display_name=Visibility
       sections [{type=MovieSceneByteSection, channels [None {type=MovieSceneScriptingByteChannel, keys [[25,3],[400,1]]}]}]
      }
     ]
    }
    XPAmountText {
     id=#13
     possessed_class=CommonTextBlock
     tracks [{
       type=MovieScene2DTransformTrack
       sections [{
         type=MovieScene2DTransformSection
         channels [
          Translation.X {type=MovieSceneScriptingFloatChannel}
          Translation.Y {type=MovieSceneScriptingFloatChannel, const [0,3]}
          Angle {type=MovieSceneScriptingFloatChannel}
          Scale.X {type=MovieSceneScriptingFloatChannel, const [1,3]}
          Scale.Y {type=MovieSceneScriptingFloatChannel, const [1,3]}
          Shear.X {type=MovieSceneScriptingFloatChannel}
          Shear.Y {type=MovieSceneScriptingFloatChannel}
         ]
        }]
      }]
    }
    XPAmountTextList {
     id=#14
     possessed_class=CommonTextBlock
     tracks [{
       type=MovieScene2DTransformTrack
       sections [{
         type=MovieScene2DTransformSection
         channels [
          Translation.X {type=MovieSceneScriptingFloatChannel}
          Translation.Y {type=MovieSceneScriptingFloatChannel, const [0,3]}
          Angle {type=MovieSceneScriptingFloatChannel}
          Scale.X {type=MovieSceneScriptingFloatChannel, const [1,3]}
          Scale.Y {type=MovieSceneScriptingFloatChannel, const [1,3]}
          Shear.X {type=MovieSceneScriptingFloatChannel}
          Shear.Y {type=MovieSceneScriptingFloatChannel}
         ]
        }]
      }]
    }
    BoostTriangles {
     id=#15
     possessed_class=Image
     tracks [
      {
       type=MovieSceneByteTrack
       display_name=Visibility
       sections [{type=MovieSceneByteSection, channels [None {type=MovieSceneScriptingByteChannel, keys [[25,3],[400,1]]}]}]
      }
      {
       type=MovieSceneFloatTrack
       display_name=Render Opacity
       sections [{type=MovieSceneFloatSection, channels [None {type=MovieSceneScriptingFloatChannel, keys [[25,0],[65,0.45],[400,0]]}]}]
      }
     ]
    }
    BoostGlow {
     id=#16
     possessed_class=Image
     tracks [
      {
       type=MovieSceneFloatTrack
       display_name=Render Opacity
       sections [{type=MovieSceneFloatSection, channels [None {type=MovieSceneScriptingFloatChannel, keys [[0,0],[25,0],[65,0.45],[400,0],[700,0]]}]}]
      }
      {
       type=MovieSceneByteTrack
       display_name=Visibility
       sections [{type=MovieSceneByteSection, channels [None {type=MovieSceneScriptingByteChannel, keys [[25,3],[400,1]]}]}]
      }
     ]
    }
   ]
   display_rate {numerator=100}
  }
 }
 OnDeltaProgressFilledToCompleteToast {
  binding_count=7
  length_seconds=1.88
  object=&self:OnDeltaProgressFilledToCompleteToast
  sequence {
   bindings [
    ChallengeProgressBar {
     id=#3
     possessed_class=Image
     tracks [
      {
       type=MovieSceneWidgetMaterialTrack
       sections [{
         type=MovieSceneParameterSection
         channels [
          AnimateDeltaProgress {type=MovieSceneScriptingFloatChannel, keys [[96,0],[152,1]]}
          AnimateProgressBarColor {type=MovieSceneScriptingFloatChannel, keys [[160,0],[168,1]]}
          Progress-Alpha {type=MovieSceneScriptingFloatChannel, keys [[160,1],[168,0]]}
         ]
        }]
      }
      {
       type=MovieScene2DTransformTrack
       sections [{
         type=MovieScene2DTransformSection
         channels [
          Translation.X {type=MovieSceneScriptingFloatChannel}
          Translation.Y {type=MovieSceneScriptingFloatChannel}
          Angle {type=MovieSceneScriptingFloatChannel}
          Scale.X {type=MovieSceneScriptingFloatChannel, const [1,2]}
          Scale.Y {type=MovieSceneScriptingFloatChannel, const [1,2]}
          Shear.X {type=MovieSceneScriptingFloatChannel}
          Shear.Y {type=MovieSceneScriptingFloatChannel}
         ]
        }]
      }
     ]
    }
    CompleteCheck {
     id=#5
     possessed_class=Image
     tracks [
      {
       type=MovieSceneByteTrack
       display_name=Visibility
       sections [{type=MovieSceneByteSection, channels [None {type=MovieSceneScriptingByteChannel, const [3,1]}]}]
      }
      {
       type=MovieSceneFloatTrack
       display_name=Render Opacity
       sections [{type=MovieSceneFloatSection, channels [None {type=MovieSceneScriptingFloatChannel, keys [[162,0],[172,1]]}]}]
      }
      {
       type=MovieScene2DTransformTrack
       sections [{
         type=MovieScene2DTransformSection
         channels [
          Translation.X {type=MovieSceneScriptingFloatChannel}
          Translation.Y {type=MovieSceneScriptingFloatChannel}
          Angle {type=MovieSceneScriptingFloatChannel}
          Scale.X {type=MovieSceneScriptingFloatChannel, const [1,2]}
          Scale.Y {type=MovieSceneScriptingFloatChannel, const [1,2]}
          Shear.X {type=MovieSceneScriptingFloatChannel}
          Shear.Y {type=MovieSceneScriptingFloatChannel}
         ]
        }]
      }
     ]
    }
    QuestIcon {
     id=#6
     possessed_class=FortLazyImage
     tracks [
      {
       type=MovieScene2DTransformTrack
       sections [{
         type=MovieScene2DTransformSection
         channels [
          Translation.X {type=MovieSceneScriptingFloatChannel}
          Translation.Y {type=MovieSceneScriptingFloatChannel}
          Angle {type=MovieSceneScriptingFloatChannel}
          Scale.X {type=MovieSceneScriptingFloatChannel, keys [[152,1],[162,0.5]]}
          Scale.Y {type=MovieSceneScriptingFloatChannel, keys [[152,1],[162,0.5]]}
          Shear.X {type=MovieSceneScriptingFloatChannel}
          Shear.Y {type=MovieSceneScriptingFloatChannel}
         ]
        }]
      }
      {
       type=MovieSceneFloatTrack
       display_name=Render Opacity
       sections [{type=MovieSceneFloatSection, channels [None {type=MovieSceneScriptingFloatChannel, keys [[152,1],[162,0]]}]}]
      }
     ]
    }
    RewardVB {
     id=#17
     possessed_class=VerticalBox
     tracks [{
       type=MovieSceneByteTrack
       display_name=Visibility
       sections [{type=MovieSceneByteSection, channels [None {type=MovieSceneScriptingByteChannel, const [1,1]}]}]
      }]
    }
    AnimBoundCompleteBorder {
     id=#18
     possessed_class=Border
     tracks [
      {
       type=MovieSceneFloatVectorTrack
       display_name=Desired Size Scale
       sections [{
         type=MovieSceneFloatVectorSection
         channels [
          Vector.X {type=MovieSceneScriptingFloatChannel}
          Vector.Y {type=MovieSceneScriptingFloatChannel, keys [[152,0],[168,1]]}
         ]
        }]
      }
      {
       type=MovieSceneFloatTrack
       display_name=Render Opacity
       sections [{type=MovieSceneFloatSection, channels [None {type=MovieSceneScriptingFloatChannel, keys [[0.0017,0],[160,0],[168,1]]}]}]
      }
      {
       type=MovieScene2DTransformTrack
       sections [{
         type=MovieScene2DTransformSection
         channels [
          Translation.X {type=MovieSceneScriptingFloatChannel, keys [[160,-16],[168,0]]}
          Translation.Y {type=MovieSceneScriptingFloatChannel, keys [[160,-4],[168,0]]}
          Angle {type=MovieSceneScriptingFloatChannel}
          Scale.X {type=MovieSceneScriptingFloatChannel}
          Scale.Y {type=MovieSceneScriptingFloatChannel}
          Shear.X {type=MovieSceneScriptingFloatChannel}
          Shear.Y {type=MovieSceneScriptingFloatChannel}
         ]
        }]
      }
     ]
    }
    RightSideInfoVB {
     id=#19
     possessed_class=VerticalBox
     tracks [
      {
       type=MovieSceneByteTrack
       display_name=Visibility
       sections [{type=MovieSceneByteSection, channels [None {type=MovieSceneScriptingByteChannel, const [3,1]}]}]
      }
      {
       type=MovieScene2DTransformTrack
       sections [{
         type=MovieScene2DTransformSection
         channels [
          Translation.X {type=MovieSceneScriptingFloatChannel, keys [[0.0017,-32],[20,0]]}
          Translation.Y {type=MovieSceneScriptingFloatChannel}
          Angle {type=MovieSceneScriptingFloatChannel}
          Scale.X {type=MovieSceneScriptingFloatChannel}
          Scale.Y {type=MovieSceneScriptingFloatChannel}
          Shear.X {type=MovieSceneScriptingFloatChannel}
          Shear.Y {type=MovieSceneScriptingFloatChannel}
         ]
        }]
      }
     ]
    }
    Burst {
     id=#7
     possessed_class=Image
     tracks [
      {
       type=MovieSceneColorTrack
       display_name=Color and Opacity
       sections [{
         type=MovieSceneColorSection
         channels [
          Color.R {type=MovieSceneScriptingFloatChannel, keys [[50,0.0331],[86,0.0331],[152,0.0343]]}
          Color.G {type=MovieSceneScriptingFloatChannel, keys [[50,0.855],[86,0.855],[152,1]]}
          Color.B {type=MovieSceneScriptingFloatChannel, keys [[50,1],[86,1],[152,0.227]]}
          Color.A {type=MovieSceneScriptingFloatChannel, const [1,1]}
         ]
        }]
      }
      {
       type=MovieSceneByteTrack
       display_name=Visibility
       sections [{type=MovieSceneByteSection, channels [None {type=MovieSceneScriptingByteChannel, keys [[50,3],[188,1]]}]}]
      }
      {
       type=MovieSceneFloatTrack
       display_name=Render Opacity
       sections [{
         type=MovieSceneFloatSection
         channels [None {type=MovieSceneScriptingFloatChannel, keys [[0.0017,0],[50,0],[60,0.2],[65,0.2],[86,0],[152,0],[162,0.2],[167,0.2],[188,0]]}]
        }]
      }
      {
       type=MovieScene2DTransformTrack
       sections [{
         type=MovieScene2DTransformSection
         channels [
          Translation.X {type=MovieSceneScriptingFloatChannel}
          Translation.Y {type=MovieSceneScriptingFloatChannel}
          Angle {type=MovieSceneScriptingFloatChannel}
          Scale.X {type=MovieSceneScriptingFloatChannel, keys [[50,1],[86,2],[152,1],[188,2]]}
          Scale.Y {type=MovieSceneScriptingFloatChannel, keys [[50,1],[86,2],[152,1],[188,2]]}
          Shear.X {type=MovieSceneScriptingFloatChannel}
          Shear.Y {type=MovieSceneScriptingFloatChannel}
         ]
        }]
      }
     ]
    }
   ]
   display_rate {numerator=100}
   tracks [{
     type=MovieSceneEventTrack
     sections [{
       type=MovieSceneEventTriggerSection
       channels [None {type=MovieSceneScriptingEventChannel, const ["<Struct 'MovieSceneEvent' (0x46742a4c260) {}>", 1]}]
      }]
    }]
  }
 }
]
graphs [
 EventGraph {nodes [
   OnToastFinished() {is=title K2Node_CustomEvent}
   > Call Toast Finished!(execute:Exec, self:Mini Challenge Entry Object Reference) -> (then:Exec) {is=title K2Node_CallDelegate}
   n1 = ResetProgressData() {is=title K2Node_CustomEvent}
   n2 = Sequence(execute=[@n1.then, @n47.then])
   > n3 = GetDynamicMaterial(self=Get ChallengeProgressBar() -> (ChallengeProgressBar:Image Object Reference))
   n4 = Get LastKnownCount() -> (LastKnownCount:int=0)
   n5 = Get GoalCount() -> (GoalCount:int=0)
   > SetScalarParameterValue(
    self=@n3
    ParameterName=Progress
    Value=float / float(A="To Float (Integer)"(InInt=@n4), B="To Float (Integer)"(InInt=@n5)) {is=title K2Node_CallFunction}
   )
   > SetText(
    self=Get CountText() -> (CountText:Common Rich Text Block Object Reference)
    InText=Format Text!(
     Format="LOCTEXT(\"<text color=\\\"ffffff\\\" fontface=\\\"black\\\">{LastKnownCount}</> / {GoalCount}\")"
     LastKnownCount:int=Get AchivedCount() -> (AchivedCount:int=0)
     GoalCount:int=Get GoalCount() -> (GoalCount=0)
    ) -> (Result)
   )
   n6 = Get AchivedCount() -> (AchivedCount=0)
   n7 = Get AchivedCount() -> (AchivedCount=0)
   n8 = Get LastKnownCount() -> (LastKnownCount=0)
   n9 = Branch(
    execute=@n2.then_1
    Condition=AND Boolean(
     A="Not Equal (Integer)"!(A:int=@n8, B:int=@n7) -> (ReturnValue:bool) {is=title K2Node_CallFunction}
     B="integer > integer"(A=@n6, B=0) {is=title K2Node_CallFunction}
    )
   )
   > n10 = GetDynamicMaterial(self=Get ChallengeProgressBar())
   n11 = Get AchivedCount() -> (AchivedCount=0)
   n12 = Get GoalCount() -> (GoalCount=0)
   > SetScalarParameterValue(self=@n10, ParameterName=DeltaProgress, Value=float / float(A="To Float (Integer)"(InInt=@n11), B="To Float (Integer)"(InInt=@n12)))
   n13 = Get OnDeltaProgressFilled() -> (OnDeltaProgressFilled:Widget Animation Object Reference)
   > PlayAnimationForward(InAnimation=@n13)
   > BindToAnimationFinished(Animation=@n13, Delegate=Create Event())
   n14 = GetDynamicMaterial(execute=@n9.else, self=Get ChallengeProgressBar())
   > SetScalarParameterValue(self=@n14, ParameterName=DeltaProgress, Value=0)
   Play Fill to Complete Post() {is=title K2Node_CustomEvent}
   > PlaySound2D(Sound=/Game/Athena/Sounds/Quests/MiniChallenges/XP_Min-Challenges_Post_Cue)
   CompetionToast() {is=title K2Node_CustomEvent}
   n15 = Get OnDeltaProgressFilledToCompleteToast() -> (OnDeltaProgressFilledToCompleteToast:Widget Animation Object Reference)
   > PlayAnimationForward(InAnimation=@n15)
   > BindToAnimationFinished(Animation=@n15, Delegate=Create Event())
   Play Refresh to New() {is=title K2Node_CustomEvent}
   > PlaySound2D(Sound=/Game/Athena/Sounds/Quests/MiniChallenges/XP_Min-Challenges_Reappear_Cue)
   n16 = OnDeltaProgressFinished() {is=title K2Node_CustomEvent}
   FinishedQuestUpdate(execute=[@n16.then, @n17.then], self:Athena Mini Challenge Entry Object Reference)
   n17 = OnDeltaProgressToCompleteFinished() {is=title K2Node_CustomEvent}
   Event TriggerNewQuestGranted() {is=title K2Node_Event}
   > PlayAnimationForward(InAnimation=Get OnCompleteRefreshToNew() -> (OnCompleteRefreshToNew:Widget Animation Object Reference))
   > StopAnimation(InAnimation=Get OnDeltaProgressFilledToComplete() -> (OnDeltaProgressFilledToComplete:Widget Animation Object Reference))
   Event Construct()
   n18 = Get ChallengeProgressBar()
   n19 = Break Slate Brush!(SlateBrush:"Slate Brush Structure (by ref)"=Get Brush!(self:Image Object Reference=@n18) -> (Brush:Slate Brush Structure)) -> (
    TintColor:Slate Color Structure
    DrawAs:ESlateBrushDrawType Enum
    Tiling:ESlateBrushTileType Enum
    Mirroring:ESlateBrushMirrorType Enum
    ImageSize:Vector 2D Structure
    Margin:Margin Structure
    ResourceObject:Object Reference
    OutlineSettings:Slate Brush Outline Settings Structure
   ) {is=title K2Node_BreakStruct}
   n20 = IsInZone!(self:Fort Global UIContext Object Reference=Get FortGlobalUIContext() -> (ReturnValue:Fort Global UIContext Object Reference) {is=title K2Node_GetSubsystem}) -> (ReturnValue:bool)
   > SetBrush(
    self=@n18
    InBrush=Make Slate Brush!(
     TintColor:Slate Color Structure=@n19
     DrawAs:ESlateBrushDrawType Enum=@n19.DrawAs
     Tiling:ESlateBrushTileType Enum=@n19.Tiling
     Mirroring:ESlateBrushMirrorType Enum=@n19.Mirroring
     ImageSize:Vector 2D Structure=@n19.ImageSize
     Margin:Margin Structure=@n19.Margin
     ResourceObject:Object Reference=Select(
      Option 0:Object Reference=/Game/Athena/UI/Challenges/Art/MiniChallenges/M_UI_MiniChallenges_Frontend_Icon
      Option 1:Object Reference=/Game/Athena/UI/Challenges/Art/MiniChallenges/M_UI_MiniChallenges_Icon
      Index=@n20
     ) -> (ReturnValue:Object Reference)
     OutlineSettings:Slate Brush Outline Settings Structure=(CornerRadii=(X=0,Y=0,Z=0,W=0),Color=(SpecifiedColor=(R=0,G=0,B=0,A=0),ColorUseRule=UseColor_Specified),Width=0,RoundingType=HalfHeightRadius,bUseBrushTransparency=False)
    ) -> (SlateBrush:Slate Brush Structure) {is=title K2Node_MakeStruct}
   )
   > n21 = Branch(Condition=Get ProgressBarUpdated() -> (ProgressBarUpdated:bool))
   > n22 = ResetProgressData() {is=title K2Node_CallFunction}
   n23 = Select(Option 0:double=64, Option 1:double=76, Index=@n20) -> (ReturnValue:double)
   n24 = Get ProgressBarSB() -> (ProgressBarSB:Size Box Object Reference)
   SetWidthOverride(execute=[@n22.then, @n21.else], self=@n24, InWidthOverride=@n23)
   > SetHeightOverride(self=@n24, InHeightOverride=@n23)
   n25 = Select(Option 0=44, Option 1=52, Index=@n20)
   n26 = Get IconSB() -> (IconSB:Size Box Object Reference)
   > SetWidthOverride(self=@n26, InWidthOverride=@n25)
   > SetHeightOverride(self=@n26, InHeightOverride=@n25)
   > n27 = Branch(Condition=@n20)
   n28 = Get RightSideInfoVB() -> (RightSideInfoVB:Vertical Box Object Reference)
   n29 = Get RewardOvr() -> (RewardOvr:Overlay Object Reference)
   n30 = Get CountText()
   > SetVisibility(self=[@n28, @n29, @n30], InVisibility=Collapsed)
   n31 = Get AnimBoundCompleteBorder() -> (AnimBoundCompleteBorder:Border Object Reference)
   > n32 = SetVisibility(self=@n31, InVisibility=HitTestInvisible)
   BindToAnimationStarted!(
    execute=[@n32.then, @n33.then]
    self:User Widget Object Reference
    Animation:Widget Animation Object Reference=Get OnDeltaProgressFilledToComplete()
    Delegate:Delegate=Create Event()
   ) -> (then)
   > BindToAnimationStarted(Animation=Get OnCompleteRefreshToNew(), Delegate=Create Event())
   SetVisibility(execute=@n27.else, self=[@n28, @n29, @n30], InVisibility=HitTestInvisible)
   > n33 = SetVisibility(self=@n31, InVisibility=Collapsed)
   Event PreConstruct()
   > Branch(Condition=Get NotificationDisplay() -> (NotificationDisplay:bool))
   > SetVisibility(self=[Get RightSideInfoVB(), Get RewardOvr(), Get CountText()], InVisibility=HitTestInvisible)
   > SetVisibility(self=Get RewardVB() -> (RewardVB:Vertical Box Object Reference), InVisibility=Collapsed)
   n34 = Event SetXpValue!() -> (OutputDelegate, then, XpRewardValue:int=0) {is=title K2Node_Event}
   > n35 = Branch(Condition=IsInZone(self=Get FortGlobalUIContext()))
   n36 = Get RewardVB()
   > n37 = SetVisibility(self=@n36, InVisibility=SelfHitTestInvisible)
   n38 = Branch(execute=[@n45.then, @n37.then], Condition=Get bIsFirstTimeBoosted() -> (bIsFirstTimeBoosted:bool))
   > n39 = StyleAsFirstTimeBoosted()
   n40 = Branch(execute=[@n39.then, @n44.then], Condition="integer > integer"(A=@n34.XpRewardValue, B=1000))
   n41 = "% (Integer)"(A=@n34.XpRewardValue, B=1000) -> (ReturnValue=0)
   > n42 = Branch(Condition="integer > integer"(A=@n41, B=0))
   n43 = int / int!(A:int=@n34.XpRewardValue, B:int=1000) -> (ReturnValue:int=0) {is=title K2Node_CallFunction}
   > "SetText (Text)"(
    self=[Get XPAmountText() -> (XPAmountText:Common Text Object Reference), Get XPAmountTextList() -> (XPAmountTextList:Common Text Object Reference)]
    InText=Format Text!(Format="LOCTEXT(\"{1000\\'s}.{100\\'s}K\")", 1000's:int=@n43, 100's:int=int / int(A=@n41, B=100) -> (ReturnValue=0)) -> (Result)
   )
   "SetText (Text)"(
    execute=@n42.else
    self=[Get XPAmountText(), Get XPAmountTextList()]
    InText=Format Text!(Format="LOCTEXT(\"{Value}K\")", Value:int=@n43) -> (Result)
   )
   "SetText (Text)"(
    execute=@n40.else
    self=[Get XPAmountText(), Get XPAmountTextList()]
    InText=Format Text(Format="LOCTEXT(\"{Value}\")", Value=@n34.XpRewardValue)
   )
   n44 = StyleAsNormal(execute=@n38.else)
   n45 = SetVisibility(execute=@n35.else, self=@n36, InVisibility=Collapsed)
   n46 = Event UpdateProgressBar!() -> (OutputDelegate, then, LastKnownCount:int=0, AchivedCount:int=0, GoalCount:int=0) {is=title K2Node_Event}
   > Set LastKnownCount(LastKnownCount:int=@n46.LastKnownCount) -> (Output_Get:int=0)
   > Set AchivedCount(AchivedCount:int=@n46.AchivedCount) -> (Output_Get:int=0)
   > Set GoalCount(GoalCount:int=@n46.GoalCount) -> (Output_Get:int=0)
   > n47 = Set ProgressBarUpdated(ProgressBarUpdated:bool=true) -> (Output_Get:bool)
   n48 = Event UpdateQuestDescription!() -> (OutputDelegate, then, Description:"Text (by ref)") {is=title K2Node_Event}
   > "SetText (Text)"(self=Get ChallengeDescription() -> (ChallengeDescription:Common Text Object Reference), InText=@n48.Description)
   Event QuestCompletedFinishFill() {is=title K2Node_Event}
   > n49 = GetDynamicMaterial(self=Get ChallengeProgressBar())
   > SetScalarParameterValue(self=@n49, ParameterName=DeltaProgress, Value=1)
   n50 = Get OnDeltaProgressFilledToComplete()
   > PlayAnimationForward(InAnimation=@n50)
   > BindToAnimationFinished(Animation=@n50, Delegate=Create Event())
   BindToAnimationStarted(Animation=Get OnDeltaProgressFilledToCompleteToast(), Delegate=Create Event())
   FadeToastAway() {is=title K2Node_CustomEvent}
  ]}
 Sequencer Events {nodes [
   SequenceEvent_0() {is=title K2Node_CustomEvent}
   > Call Trigger Xp Toast!(execute:Exec, self:Mini Challenge Entry Object Reference) -> (then:Exec) {is=title K2Node_CallDelegate}
  ]}
 StyleAsFirstTimeBoosted {nodes [
   StyleAsFirstTimeBoosted() {is=title K2Node_FunctionEntry}
   > SetColorAndOpacity(
    self=[Get XPTextList() -> (XPTextList:Common Text Object Reference), Get XPText() -> (XPText:Common Text Object Reference)]
    InColorAndOpacity=Make Slate Color(SpecifiedColor=(R=0.9301,G=1,B=0.0097,A=1))
   )
   > SetColorAndOpacity(self=[Get XPAmountText(), Get XPAmountTextList()], InColorAndOpacity=Make Slate Color(SpecifiedColor=(R=0.9301,G=1,B=0.0097,A=1)))
   > PlayAnimation(InAnimation=Get TransientXPBoostedBurst() -> (TransientXPBoostedBurst:Widget Animation Object Reference))
  ]}
 StyleAsNormal {nodes [
   StyleAsNormal() {is=title K2Node_FunctionEntry}
   n1 = Select(
    Option 0:Linear Color Structure=(R=0.2159,G=0.6584,B=1,A=1)
    Option 1:Linear Color Structure=(R=1,G=1,B=1,A=1)
    Index=IsInZone(self=Get FortGlobalUIContext())
   ) -> (ReturnValue:Linear Color Structure)
   > SetColorAndOpacity(self=[Get XPTextList(), Get XPText()], InColorAndOpacity=Make Slate Color(SpecifiedColor=@n1))
   > SetColorAndOpacity(self=[Get XPAmountTextList(), Get XPAmountText()], InColorAndOpacity=Make Slate Color(SpecifiedColor=@n1))
   > SetVisibility(
    self=[Get BoostGlow() -> (BoostGlow:Image Object Reference), Get BoostTriangles() -> (BoostTriangles:Image Object Reference)]
    InVisibility=Collapsed
   )
  ]}
 ToastFinished {nodes [ToastFinished() {is=title K2Node_FunctionEntry}]}
 TriggerXpToast {nodes [TriggerXpToast() {is=title K2Node_FunctionEntry}]}
]
variables [
 AchivedCount {type=integer}
 GoalCount {type=integer}
 LastKnownCount {type=integer}
 NotificationDisplay {type=boolean}
 ProgressBarUpdated {type=boolean}
 ToastFinished {}
 TriggerXpToast {}
]
widgets {
 info {inheritedWidgetCount=1, parentClass=/Script/FortniteUI.AthenaMiniChallengeEntry, rootWidgetClass=&HorizontalBox, widgetCount=29}
 widgets [
  AnimBoundHB {widgetClassPath=&HorizontalBox, widget=AnimBoundHB}
  VerticalBox_1 {widgetClassPath=&VerticalBox, parent=AnimBoundHB, widget=VerticalBox_1, slot {class=HorizontalBoxSlot}}
  ProgressBarSB {
   widgetClassPath=&SizeBox
   bIsVariable=true
   parent=VerticalBox_1
   widget=ProgressBarSB
   properties {bOverride_HeightOverride=true, bOverride_WidthOverride=true, heightOverride=76, widthOverride=76}
   slot {class=VerticalBoxSlot, properties {horizontalAlignment=HAlign_Center, padding=<ltrb 2 0 2 0>}}
  }
  Overlay_0 {widgetClassPath=&Overlay, parent=ProgressBarSB, widget=Overlay_0, slot {class=SizeBoxSlot}}
  Burst {
   widgetClassPath=&Image
   bIsVariable=true
   parent=Overlay_0
   widget=Burst
   properties {
    renderOpacity=0
    visibility=Collapsed
    brush {
     bIsDynamicallyLoaded=false
     drawAs=Image
     imageType=NoImage
     mirroring=NoMirror
     resourceName=None
     resourceObject=/Game/Athena/UI/Challenges/Art/MiniChallenges/M_UI_MiniChallenges_Burst
     tiling=NoTile
     imageSize=<32 32>
     margin=<ltrb 0 0 0 0>
     outlineSettings {
      bUseBrushTransparency=false
      roundingType=HalfHeightRadius
      width=0
      color {colorUseRule=UseColor_Specified, specifiedColor=<rgba 0 0 0 0>}
      cornerRadii=<0 0 0 0>
     }
     tintColor {colorUseRule=UseColor_Specified, specifiedColor=<rgba 1 1 1 1>}
     uVRegion {bIsValid=false, max=<0 0>, min=<0 0>}
    }
   }
   slot {class=OverlaySlot, properties {horizontalAlignment=HAlign_Fill, verticalAlignment=VAlign_Fill}}
  }
  ChallengeProgressBar {
   widgetClassPath=&Image
   bIsVariable=true
   parent=Overlay_0
   widget=ChallengeProgressBar
   properties {
    visibility=SelfHitTestInvisible
    brush {
     bIsDynamicallyLoaded=false
     drawAs=Image
     imageType=NoImage
     mirroring=NoMirror
     resourceName=None
     resourceObject=&M_UI_MiniChallenges_Icon
     tiling=NoTile
     imageSize=<32 32>
     margin=<ltrb 0 0 0 0>
     outlineSettings {
      bUseBrushTransparency=false
      roundingType=HalfHeightRadius
      width=0
      color {colorUseRule=UseColor_Specified, specifiedColor=<rgba 0 0 0 0>}
      cornerRadii=<0 0 0 0>
     }
     tintColor {colorUseRule=UseColor_Specified, specifiedColor=<rgba 1 1 1 1>}
     uVRegion {bIsValid=false, max=<0 0>, min=<0 0>}
    }
   }
   slot {class=OverlaySlot, properties {horizontalAlignment=HAlign_Fill, verticalAlignment=VAlign_Fill}}
  }
  IconSB {
   widgetClassPath=&SizeBox
   bIsVariable=true
   parent=Overlay_0
   widget=IconSB
   properties {bOverride_HeightOverride=true, bOverride_WidthOverride=true, heightOverride=52, widthOverride=52}
   slot {class=OverlaySlot, properties {horizontalAlignment=HAlign_Center, verticalAlignment=VAlign_Center}}
  }
  Overlay_1 {widgetClassPath=&Overlay, parent=IconSB, widget=Overlay_1, slot {class=SizeBoxSlot}}
  QuestIcon {
   widgetClassPath=&FortLazyImage
   bInherited=true
   bIsVariable=true
   parent=Overlay_1
   widget=QuestIcon
   properties {brush {
     bIsDynamicallyLoaded=false
     drawAs=Image
     imageType=NoImage
     mirroring=NoMirror
     resourceName=None
     resourceObject=/Game/Athena/UI/Challenges/Art/MiniChallenges/M_UI_MiniChallenges_IconMask
     tiling=NoTile
     imageSize=<32 32>
     margin=<ltrb 0 0 0 0>
     outlineSettings {
      bUseBrushTransparency=false
      roundingType=HalfHeightRadius
      width=0
      color {colorUseRule=UseColor_Specified, specifiedColor=<rgba 0 0 0 0>}
      cornerRadii=<0 0 0 0>
     }
     tintColor {colorUseRule=UseColor_Specified, specifiedColor=<rgba 1 1 1 1>}
     uVRegion {bIsValid=false, max=<0 0>, min=<0 0>}
    }}
   slot {class=OverlaySlot, properties {horizontalAlignment=HAlign_Fill, verticalAlignment=VAlign_Fill}}
  }
  CompleteCheck {
   widgetClassPath=&Image
   bIsVariable=true
   parent=Overlay_1
   widget=CompleteCheck
   properties {
    renderOpacity=0
    visibility=Collapsed
    brush {
     bIsDynamicallyLoaded=false
     drawAs=Image
     imageType=NoImage
     mirroring=NoMirror
     resourceName=None
     resourceObject=/Game/UI/Foundation/Textures/Icons/Checks/T-Icon-Check-64
     tiling=NoTile
     imageSize=<64 64>
     margin=<ltrb 0 0 0 0>
     outlineSettings {
      bUseBrushTransparency=false
      roundingType=HalfHeightRadius
      width=0
      color {colorUseRule=UseColor_Specified, specifiedColor=<rgba 0 0 0 0>}
      cornerRadii=<0 0 0 0>
     }
     tintColor {colorUseRule=UseColor_Specified, specifiedColor=<rgba 1 1 1 1>}
     uVRegion {bIsValid=false, max=<0 0>, min=<0 0>}
    }
    colorAndOpacity=<rgba 0.0343 1 0.227 1>
   }
   slot {class=OverlaySlot, properties {horizontalAlignment=HAlign_Fill, verticalAlignment=VAlign_Fill, padding=<ltrb 10 10 10 10>}}
  }
  RewardVB {
   widgetClassPath=&VerticalBox
   bIsVariable=true
   parent=VerticalBox_1
   widget=RewardVB
   slot {class=VerticalBoxSlot, properties {horizontalAlignment=HAlign_Center}}
  }
  XPAmountText {
   widgetClassPath=&CommonTextBlock
   bIsVariable=true
   parent=RewardVB
   widget=XPAmountText
   properties {
    justification=Center
    text=999k
    textTransformPolicy=ToUpper
    colorAndOpacity {colorUseRule=UseColor_Specified, specifiedColor=<rgba 0.9301 1 0.0097 1>}
    font {
     bForceMonospaced=false
     bMaterialIsStencil=false
     fontMaterial=null
     fontObject=&BurbankSmall
     letterSpacing=0
     monospacedWidth=1
     size=14
     skewAmount=0
     typefaceFontName=Black
     outlineSettings {
      bApplyOutlineToDropShadows=false
      bMiteredCorners=false
      bSeparateFillAlpha=false
      outlineMaterial=null
      outlineSize=0
      outlineColor=<rgba 0 0 0 1>
     }
    }
    renderTransform {angle=0, scale=<1 1>, shear=<-12 0>, translation=<0 0>}
    shadowOffset=<0 0>
   }
   slot {class=VerticalBoxSlot}
  }
  Overlay_2 {
   widgetClassPath=&Overlay
   parent=RewardVB
   widget=Overlay_2
   slot {class=VerticalBoxSlot, properties {horizontalAlignment=HAlign_Center, padding=<ltrb 0 -4 0 0>}}
  }
  BoostGlow {
   widgetClassPath=&Image
   bIsVariable=true
   parent=Overlay_2
   widget=BoostGlow
   properties {
    renderOpacity=0
    visibility=Collapsed
    brush {
     bIsDynamicallyLoaded=false
     drawAs=Image
     imageType=NoImage
     mirroring=NoMirror
     resourceName=None
     resourceObject=/FortUILibrary/Styles/Border/Materials/M_Border_SoftGlow
     tiling=NoTile
     imageSize=<32 32>
     margin=<ltrb 0 0 0 0>
     outlineSettings {
      bUseBrushTransparency=false
      roundingType=HalfHeightRadius
      width=0
      color {colorUseRule=UseColor_Specified, specifiedColor=<rgba 0 0 0 0>}
      cornerRadii=<0 0 0 0>
     }
     tintColor {colorUseRule=UseColor_Specified, specifiedColor=<rgba 1 1 1 1>}
     uVRegion {bIsValid=false, max=<0 0>, min=<0 0>}
    }
    colorAndOpacity=<rgba 1 0.6038 0 1>
   }
   slot {class=OverlaySlot, properties {horizontalAlignment=HAlign_Fill, verticalAlignment=VAlign_Fill, padding=<ltrb -32 -32 -32 -16>}}
  }
  BoostTriangles {
   widgetClassPath=&Image
   bIsVariable=true
   parent=Overlay_2
   widget=BoostTriangles
   properties {
    renderOpacity=0
    visibility=Collapsed
    brush {
     bIsDynamicallyLoaded=false
     drawAs=Image
     imageType=NoImage
     mirroring=NoMirror
     resourceName=None
     resourceObject=/Game/Athena/UI/Challenges/Art/MiniChallenges/M_UI_MiniChallenges_Boost
     tiling=NoTile
     imageSize=<32 32>
     margin=<ltrb 0 0 0 0>
     outlineSettings {
      bUseBrushTransparency=false
      roundingType=HalfHeightRadius
      width=0
      color {colorUseRule=UseColor_Specified, specifiedColor=<rgba 0 0 0 0>}
      cornerRadii=<0 0 0 0>
     }
     tintColor {colorUseRule=UseColor_Specified, specifiedColor=<rgba 1 1 1 1>}
     uVRegion {bIsValid=false, max=<0 0>, min=<0 0>}
    }
   }
   slot {class=OverlaySlot, properties {horizontalAlignment=HAlign_Fill, verticalAlignment=VAlign_Fill, padding=<ltrb -16 -24 -16 0>}}
  }
  XPText {
   widgetClassPath=&CommonTextBlock
   bIsVariable=true
   parent=Overlay_2
   widget=XPText
   properties {
    justification=Center
    text=xp
    textTransformPolicy=ToUpper
    colorAndOpacity {colorUseRule=UseColor_Specified, specifiedColor=<rgba 0.9301 1 0.0097 1>}
    font {
     bForceMonospaced=false
     bMaterialIsStencil=false
     fontMaterial=null
     fontObject=&BurbankSmall
     letterSpacing=0
     monospacedWidth=1
     size=12
     skewAmount=0
     typefaceFontName=Black
     outlineSettings {
      bApplyOutlineToDropShadows=false
      bMiteredCorners=false
      bSeparateFillAlpha=false
      outlineMaterial=null
      outlineSize=0
      outlineColor=<rgba 0 0 0 1>
     }
    }
    renderTransform {angle=0, scale=<1 1>, shear=<-12 0>, translation=<0 0>}
    shadowOffset=<0 0>
   }
   slot {class=OverlaySlot, properties {horizontalAlignment=HAlign_Fill, verticalAlignment=VAlign_Fill}}
  }
  RightSideInfoVB {
   widgetClassPath=&VerticalBox
   bIsVariable=true
   parent=AnimBoundHB
   widget=RightSideInfoVB
   properties {visibility=Collapsed}
   slot {class=HorizontalBoxSlot, properties {verticalAlignment=VAlign_Center, padding=<ltrb 8 0 0 0>, size {sizeRule=Fill, value=1}}}
  }
  HorizontalBox_50 {
   widgetClassPath=&HorizontalBox
   parent=RightSideInfoVB
   widget=HorizontalBox_50
   slot {class=VerticalBoxSlot, properties {size {sizeRule=Fill, value=1}}}
  }
  ChallengeDescription {
   widgetClassPath=&CommonTextBlock
   bIsVariable=true
   parent=HorizontalBox_50
   widget=ChallengeDescription
   properties {
    scrollStyle=&TextScrollStyle-Base_C
    text=Challenge description
    font {
     bForceMonospaced=false
     bMaterialIsStencil=false
     fontMaterial=null
     fontObject=&BurbankSmall
     letterSpacing=0
     monospacedWidth=1
     size=14
     skewAmount=0
     typefaceFontName=Bold
     outlineSettings {
      bApplyOutlineToDropShadows=false
      bMiteredCorners=false
      bSeparateFillAlpha=false
      outlineMaterial=null
      outlineSize=0
      outlineColor=<rgba 0 0 0 1>
     }
    }
    renderTransform {angle=0, scale=<1 1>, shear=<-12 0>, translation=<0 0>}
    shadowOffset=<0 0>
   }
   slot {class=HorizontalBoxSlot, properties {verticalAlignment=VAlign_Center, size {sizeRule=Fill, value=1}}}
  }
  AnimBoundCompleteBorder {
   widgetClassPath=&Border
   bIsVariable=true
   parent=RightSideInfoVB
   widget=AnimBoundCompleteBorder
   properties {
    visibility=SelfHitTestInvisible
    background {
     bIsDynamicallyLoaded=false
     drawAs=NoDrawType
     imageType=NoImage
     mirroring=NoMirror
     resourceName=None
     resourceObject=null
     tiling=NoTile
     imageSize=<32 32>
     margin=<ltrb 0 0 0 0>
     outlineSettings {
      bUseBrushTransparency=false
      roundingType=HalfHeightRadius
      width=0
      color {colorUseRule=UseColor_Specified, specifiedColor=<rgba 0 0 0 0>}
      cornerRadii=<0 0 0 0>
     }
     tintColor {colorUseRule=UseColor_Specified, specifiedColor=<rgba 1 1 1 1>}
     uVRegion {bIsValid=false, max=<0 0>, min=<0 0>}
    }
    padding=<ltrb 0 0 0 0>
   }
   slot {class=VerticalBoxSlot}
  }
  CommonTextBlock {
   widgetClassPath=&CommonTextBlock
   parent=AnimBoundCompleteBorder
   widget=CommonTextBlock
   properties {
    text="Complete!"
    textTransformPolicy=ToUpper
    colorAndOpacity {colorUseRule=UseColor_Specified, specifiedColor=<rgba 0.0343 1 0.227 1>}
    font {
     bForceMonospaced=false
     bMaterialIsStencil=false
     fontMaterial=null
     fontObject=&BurbankSmall
     letterSpacing=0
     monospacedWidth=1
     size=16
     skewAmount=0
     typefaceFontName=Black
     outlineSettings {
      bApplyOutlineToDropShadows=false
      bMiteredCorners=false
      bSeparateFillAlpha=false
      outlineMaterial=null
      outlineSize=0
      outlineColor=<rgba 0 0 0 1>
     }
    }
    renderTransform {angle=0, scale=<1 1>, shear=<-12 0>, translation=<0 0>}
    shadowOffset=<0 0>
   }
   slot {class=BorderSlot, properties {padding=<ltrb 0 0 0 0>}}
  }
  HorizontalBox_Ovr {widgetClassPath=&HorizontalBox, bIsVariable=true, parent=RightSideInfoVB, widget=HorizontalBox_Ovr, slot {class=VerticalBoxSlot}}
  RewardOvr {
   widgetClassPath=&Overlay
   bIsVariable=true
   parent=HorizontalBox_Ovr
   widget=RewardOvr
   slot {class=HorizontalBoxSlot, properties {horizontalAlignment=HAlign_Right}}
  }
  BoostGlowList {
   widgetClassPath=&Image
   bIsVariable=true
   parent=RewardOvr
   widget=BoostGlowList
   properties {
    renderOpacity=0
    visibility=Collapsed
    brush {
     bIsDynamicallyLoaded=false
     drawAs=Image
     imageType=NoImage
     mirroring=NoMirror
     resourceName=None
     resourceObject=&M_Border_SoftGlow
     tiling=NoTile
     imageSize=<16 16>
     margin=<ltrb 0 0 0 0>
     outlineSettings {
      bUseBrushTransparency=false
      roundingType=HalfHeightRadius
      width=0
      color {colorUseRule=UseColor_Specified, specifiedColor=<rgba 0 0 0 0>}
      cornerRadii=<0 0 0 0>
     }
     tintColor {colorUseRule=UseColor_Specified, specifiedColor=<rgba 1 1 1 1>}
     uVRegion {bIsValid=false, max=<0 0>, min=<0 0>}
    }
    colorAndOpacity=<rgba 1 0.6038 0 1>
   }
   slot {class=OverlaySlot, properties {horizontalAlignment=HAlign_Fill, verticalAlignment=VAlign_Fill, padding=<ltrb -32 -24 -32 -16>}}
  }
  BoostTrianglesList {
   widgetClassPath=&Image
   bIsVariable=true
   parent=RewardOvr
   widget=BoostTrianglesList
   properties {
    renderOpacity=0
    visibility=Collapsed
    brush {
     bIsDynamicallyLoaded=false
     drawAs=Image
     imageType=NoImage
     mirroring=NoMirror
     resourceName=None
     resourceObject=&M_UI_MiniChallenges_Boost
     tiling=NoTile
     imageSize=<16 16>
     margin=<ltrb 0 0 0 0>
     outlineSettings {
      bUseBrushTransparency=false
      roundingType=HalfHeightRadius
      width=0
      color {colorUseRule=UseColor_Specified, specifiedColor=<rgba 0 0 0 0>}
      cornerRadii=<0 0 0 0>
     }
     tintColor {colorUseRule=UseColor_Specified, specifiedColor=<rgba 1 1 1 1>}
     uVRegion {bIsValid=false, max=<0 0>, min=<0 0>}
    }
   }
   slot {class=OverlaySlot, properties {horizontalAlignment=HAlign_Fill, verticalAlignment=VAlign_Fill, padding=<ltrb -8 -8 -8 -8>}}
  }
  RewardHB {widgetClassPath=&HorizontalBox, parent=RewardOvr, widget=RewardHB, slot {class=OverlaySlot}}
  XPAmountTextList {
   widgetClassPath=&CommonTextBlock
   bIsVariable=true
   parent=RewardHB
   widget=XPAmountTextList
   properties {
    justification=Center
    text=999k
    textTransformPolicy=ToUpper
    colorAndOpacity {colorUseRule=UseColor_Specified, specifiedColor=<rgba 0.9301 1 0.0097 1>}
    font {
     bForceMonospaced=false
     bMaterialIsStencil=false
     fontMaterial=null
     fontObject=&BurbankSmall
     letterSpacing=0
     monospacedWidth=1
     size=14
     skewAmount=0
     typefaceFontName=Black
     outlineSettings {
      bApplyOutlineToDropShadows=false
      bMiteredCorners=false
      bSeparateFillAlpha=false
      outlineMaterial=null
      outlineSize=0
      outlineColor=<rgba 0 0 0 1>
     }
    }
    renderTransform {angle=0, scale=<1 1>, shear=<-12 0>, translation=<0 0>}
    shadowOffset=<0 0>
   }
   slot {class=HorizontalBoxSlot, properties {verticalAlignment=VAlign_Center, size {sizeRule=Fill, value=1}}}
  }
  XPTextList {
   widgetClassPath=&CommonTextBlock
   bIsVariable=true
   parent=RewardHB
   widget=XPTextList
   properties {
    justification=Center
    lineHeightPercentage=0.9
    text=xp
    textTransformPolicy=ToUpper
    colorAndOpacity {colorUseRule=UseColor_Specified, specifiedColor=<rgba 0.9301 1 0.0097 1>}
    font {
     bForceMonospaced=false
     bMaterialIsStencil=false
     fontMaterial=null
     fontObject=&BurbankSmall
     letterSpacing=0
     monospacedWidth=1
     size=12
     skewAmount=0
     typefaceFontName=Black
     outlineSettings {
      bApplyOutlineToDropShadows=false
      bMiteredCorners=false
      bSeparateFillAlpha=false
      outlineMaterial=null
      outlineSize=0
      outlineColor=<rgba 0 0 0 1>
     }
    }
    renderTransform {angle=0, scale=<1 1>, shear=<-12 0>, translation=<0 0>}
    shadowOffset=<0 0>
   }
   slot {class=HorizontalBoxSlot, properties {verticalAlignment=VAlign_Center, padding=<ltrb 4 0 0 0>}}
  }
  CountText {
   widgetClassPath=&CommonRichTextBlock
   bIsVariable=true
   parent=HorizontalBox_Ovr
   widget=CountText
   properties {
    bOverrideDefaultStyle=true
    justification=Right
    text=99 / 99
    defaultTextStyleOverride {
     overflowPolicy=Clip
     transformPolicy=None
     colorAndOpacity {colorUseRule=UseColor_Specified, specifiedColor=<rgba 0.2159 0.6584 1 1>}
     font {
      bForceMonospaced=false
      bMaterialIsStencil=false
      fontMaterial=null
      fontObject=&BurbankSmall
      letterSpacing=0
      monospacedWidth=1
      size=15
      skewAmount=0
      typefaceFontName=Bold
      outlineSettings {
       bApplyOutlineToDropShadows=false
       bMiteredCorners=false
       bSeparateFillAlpha=false
       outlineMaterial=null
       outlineSize=0
       outlineColor=<rgba 0 0 0 1>
      }
     }
     highlightColor {colorUseRule=UseColor_Specified, specifiedColor=<rgba 0 0 0 1>}
     highlightShape {
      bIsDynamicallyLoaded=false
      drawAs=Image
      imageType=NoImage
      mirroring=NoMirror
      resourceName=None
      resourceObject=null
      tiling=NoTile
      imageSize=<32 32>
      margin=<ltrb 0 0 0 0>
      outlineSettings {
       bUseBrushTransparency=false
       roundingType=HalfHeightRadius
       width=0
       color {colorUseRule=UseColor_Specified, specifiedColor=<rgba 0 0 0 0>}
       cornerRadii=<0 0 0 0>
      }
      tintColor {colorUseRule=UseColor_Specified, specifiedColor=<rgba 1 1 1 1>}
      uVRegion {bIsValid=false, max=<0 0>, min=<0 0>}
     }
     selectedBackgroundColor {colorUseRule=UseColor_Foreground, specifiedColor=<rgba 1 0 1 1>}
     shadowColorAndOpacity=<rgba 0 0 0 1>
     shadowOffset=<0 0>
     strikeBrush {
      bIsDynamicallyLoaded=false
      drawAs=Image
      imageType=NoImage
      mirroring=NoMirror
      resourceName=None
      resourceObject=null
      tiling=NoTile
      imageSize=<32 32>
      margin=<ltrb 0 0 0 0>
      outlineSettings {
       bUseBrushTransparency=false
       roundingType=HalfHeightRadius
       width=0
       color {colorUseRule=UseColor_Specified, specifiedColor=<rgba 0 0 0 0>}
       cornerRadii=<0 0 0 0>
      }
      tintColor {colorUseRule=UseColor_Specified, specifiedColor=<rgba 1 1 1 1>}
      uVRegion {bIsValid=false, max=<0 0>, min=<0 0>}
     }
     underlineBrush {
      bIsDynamicallyLoaded=false
      drawAs=Image
      imageType=NoImage
      mirroring=NoMirror
      resourceName=None
      resourceObject=null
      tiling=NoTile
      imageSize=<32 32>
      margin=<ltrb 0 0 0 0>
      outlineSettings {
       bUseBrushTransparency=false
       roundingType=HalfHeightRadius
       width=0
       color {colorUseRule=UseColor_Specified, specifiedColor=<rgba 0 0 0 0>}
       cornerRadii=<0 0 0 0>
      }
      tintColor {colorUseRule=UseColor_Specified, specifiedColor=<rgba 1 1 1 1>}
      uVRegion {bIsValid=false, max=<0 0>, min=<0 0>}
     }
    }
    renderTransform {angle=0, scale=<1 1>, shear=<-12 0>, translation=<0 0>}
   }
   slot {class=HorizontalBoxSlot, properties {verticalAlignment=VAlign_Center, padding=<ltrb 8 0 0 0>, size {sizeRule=Fill, value=1}}}
  }
 ]
}
