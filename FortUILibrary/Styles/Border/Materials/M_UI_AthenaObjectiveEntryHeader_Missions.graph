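# Translucent UI material that draws the background of an Athena objective entry header, by its name the Missions variant.
# A horizontal gradient blends two color parameters, with a rotated, aspect-adjusted triangle-grid texture mixed in more strongly along that gradient.
# A rotatable, offset UV ramp lays a third color over it as a trim band.
# A second slightly rotated ramp, scaled by Alpha, forms a sharp angled edge mask, apparently for opacity.
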Material /FortUILibrary/Styles/Border/Materials/M_UI_AthenaObjectiveEntryHeader_Missions
graph {
 properties {
  bUsedWithStaticMesh=false
  blendMode=BLEND_Translucent
  materialDomain=MD_UI
  pixelDepthOffsetMode=PDOM_Legacy
  usedShadingModels=Default Lit
 }
 nodes [
  n1 = VectorParameter() {parameterName=RGBA-2, defaultValue=<rgba 0.016 0.1746 0.7913 1>}
  n2 = VectorParameter() {parameterName=RGBA-3, defaultValue=<rgba 0 0.004 0.0723 1>}
  n3 = ComponentMask(None=Multiply(
    A=MaterialFunctionCall() {materialFunction=&GetUserInterfaceUV}.Normalized UV
    B=ScalarParameter() {defaultValue=0.75, parameterName=Tile}
   )) {r=true}
  n4 = MaterialFunctionCall() {materialFunction=&GetUserInterfaceUV}
  n5 = MaterialFunctionCall(Float2=@n4.Pixel Size) {materialFunction=&BreakOutFloat2Components}
  n6 = Max(A=@n5.G)
  n7 = Multiply(A=AppendVector(A=Divide(A=@n5.R, B=@n6), B=Constant() {r=1}), B=@n4.Normalized UV)
  n8 = ScalarParameter() {defaultValue=-0.0278, parameterName=Tri Angle}
  n9 = TextureSample(UVs=Multiply(A=Rotator(Coordinate=@n7, Time=@n8) {speed=6.2832}, B=ScalarParameter() {defaultValue=0.5, parameterName=Tri Tile Amount})) {texture=/Game/Global/Textures/Default/Masks_and_Gradients/T-TriangleGrid}
  n10 = ScalarParameter() {defaultValue=0.125, parameterName=TriangleAmount}
  n11 = TextureCoordinate()
  n12 = Rotator(Coordinate=@n11, Time=ScalarParameter() {parameterName=TrimAngle}) {speed=6.2832}
  n13 = Subtract(A=OneMinus(None=ComponentMask(None=@n12) {g=true}), B=ScalarParameter() {defaultValue=0.6, parameterName=TrimOffset})
  n14 = ScalarParameter() {defaultValue=32, parameterName=TrimFalloff}
  LinearInterpolate(
   A=LinearInterpolate(A=LinearInterpolate(A=@n1.RGB, B=@n2.RGB, Alpha=@n3), B=LinearInterpolate(A=@n1.RGB, B=@n2.RGB, Alpha=@n9.R), Alpha=Multiply(A=@n10, B=@n3))
   B=VectorParameter() {parameterName=RGBA-1, defaultValue=<rgba 0.2831 0.6376 0.9823 1>}.RGB
   Alpha=Clamp(None=Multiply(A=@n13, B=@n14))
  )
  n15 = Rotator(Coordinate=@n11, Time=ScalarParameter() {defaultValue=0.015, parameterName=EdgeAngle}) {speed=6.2832}
  n16 = Multiply(
   A=Subtract(A=ComponentMask(None=@n15) {g=true}, B=ScalarParameter() {defaultValue=0.0986, parameterName=EdgeOffset})
   B=ScalarParameter() {defaultValue=64, parameterName=EdgeFalloff}
  )
  Reroute(None=Multiply(A=Clamp(None=@n16), B=ScalarParameter() {defaultValue=1, parameterName=Alpha}))
 ]
}
