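# Sound cue that plays the mini-challenge popup UI sound in Athena quests.
# A single wave player feeds the S13_Challenge_Popup wave at reduced volume through the in-game non-critical UI sound class.
# A concurrency override allows one instance at a time, stopping the oldest when a new popup fires.
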SoundCue /Game/Athena/Sounds/Quests/MiniChallenges/Challenge_Popup_Cue
root=WavePlayer_1
nodes [{id=WavePlayer_1, properties {soundWaveAssetPtr=&/S13_Challenge_Popup}}]
properties {
 bOverrideConcurrency=true
 maxDistance=2097152
 soundClassObject=&Fort_UI_InGame_NonCritical
 volumeMultiplier=0.25
 concurrencyOverrides {maxCount=1, resolutionRule=StopOldest}
}
