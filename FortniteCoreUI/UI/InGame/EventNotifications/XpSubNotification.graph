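# XP earned sub-notification widget that shows an animated "+ N XP!" or "+ N Bars!" reward popup.
# SetRewardValue uses the reward type to select the text, icon, background and glow color; Custom entries are left empty.
# StartNotificationIntro plays the Intro slide-and-fade, whose icon track targets only the XP icon.
# Animation start/finish bindings accompany custom events for sound, opacity reset and a delayed FinishedDisplayAnim call.
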WidgetBlueprint /FortniteCoreUI/UI/InGame/EventNotifications/XpSubNotification
animations [Intro {
  binding_count=4
  length_seconds=0.5
  object=&self:Intro
  sequence {bindings [
    SizeBox_All {
     id=#1
     possessed_class=SizeBox
     tracks [
      {
       type=MovieSceneFloatTrack
       display_name=Render Opacity
       sections [{type=MovieSceneFloatSection, channels [None {type=MovieSceneScriptingFloatChannel, keys [[5,0],[8,1]]}]}]
      }
      {
       type=MovieScene2DTransformTrack
       sections [{
         type=MovieScene2DTransformSection
         channels [
          Translation.X {type=MovieSceneScriptingFloatChannel, keys [[5,-50],[8,0]]}
          Translation.Y {type=MovieSceneScriptingFloatChannel}
          Angle {type=MovieSceneScriptingFloatChannel}
          Scale.X {type=MovieSceneScriptingFloatChannel}
          Scale.Y {type=MovieSceneScriptingFloatChannel}
          Shear.X {type=MovieSceneScriptingFloatChannel}
          Shear.Y {type=MovieSceneScriptingFloatChannel}
         ]
        }]
      }
     ]
    }
    Overlay_XpiCON {
     id=#2
     possessed_class=Overlay
     tracks [
      {
       type=MovieScene2DTransformTrack
       sections [{
         type=MovieScene2DTransformSection
         channels [
          Translation.X {type=MovieSceneScriptingFloatChannel}
          Translation.Y {type=MovieSceneScriptingFloatChannel, keys [[8,28],[13,10]]}
          Angle {type=MovieSceneScriptingFloatChannel}
          Scale.X {type=MovieSceneScriptingFloatChannel}
          Scale.Y {type=MovieSceneScriptingFloatChannel}
          Shear.X {type=MovieSceneScriptingFloatChannel}
          Shear.Y {type=MovieSceneScriptingFloatChannel}
         ]
        }]
      }
      {
       type=MovieSceneFloatTrack
       display_name=Render Opacity
       sections [{type=MovieSceneFloatSection, channels [None {type=MovieSceneScriptingFloatChannel, keys [[8,0],[13,1]]}]}]
      }
     ]
    }
    Text_XpValue {
     id=#3
     possessed_class=CommonTextBlock
     tracks [
      {
       type=MovieScene2DTransformTrack
       sections [{
         type=MovieScene2DTransformSection
         channels [
          Translation.X {type=MovieSceneScriptingFloatChannel, keys [[7,-200],[11,5],[15,0]]}
          Translation.Y {type=MovieSceneScriptingFloatChannel}
          Angle {type=MovieSceneScriptingFloatChannel}
          Scale.X {type=MovieSceneScriptingFloatChannel, keys [[7,0.17],[11,1.1],[15,1]]}
          Scale.Y {type=MovieSceneScriptingFloatChannel, keys [[7,0.5],[11,0.9],[15,1]]}
          Shear.X {type=MovieSceneScriptingFloatChannel}
          Shear.Y {type=MovieSceneScriptingFloatChannel}
         ]
        }]
      }
      {
       type=MovieSceneFloatTrack
       display_name=Render Opacity
       sections [{type=MovieSceneFloatSection, channels [None {type=MovieSceneScriptingFloatChannel, keys [[5,0],[11,1]]}]}]
      }
     ]
    }
    Image_RarityColor_Glow {
     id=#4
     possessed_class=Image
     tracks [{
       type=MovieSceneFloatTrack
       display_name=Render Opacity
       sections [{type=MovieSceneFloatSection, channels [None {type=MovieSceneScriptingFloatChannel, keys [[8,0],[13,1]]}]}]
      }]
    }
   ]}
 }]
graphs [EventGraph {nodes [
   FinishedXpDisplayAnim() {is=title K2Node_CustomEvent}
   > Delay(Duration=1.4)
   > FinishedDisplayAnim(self:Xp Earned Sub Notification Object Reference)
   Reset render opacity() {is=title K2Node_CustomEvent}
   > SetRenderOpacity(self=Get SizeBox_All() -> (SizeBox_All:Size Box Object Reference), InOpacity=0)
   Play Intro Sound() {is=title K2Node_CustomEvent}
   > PlaySound2D(Sound=/Game/Athena/Sounds/Quests/MiniChallenges/Challenge_Popup_Cue)
   Event StartNotificationIntro() {is=title K2Node_Event}
   > PlayAnimationForward(InAnimation=Get Intro())
   n1 = Event SetRewardValue!() -> (OutputDelegate, then, RewardValue:int=0, RewardType:EFortEarnedSubRewardType Enum=XP) {is=title K2Node_Event}
   n2 = "To Text (Integer)"(Value=@n1.RewardValue)
   > "SetText (Text)"(
    self=Get Text_XpValue() -> (Text_XpValue:Common Text Object Reference)
    InText=Select!(
     XP:Text=Format Text!(Format="LOCTEXT(\"+ {Value} XP!\")", Value:Text=@n2) -> (Result)
     Bars:Text=Format Text(Format="LOCTEXT(\"+ {Value} Bars!\")", Value=@n2)
     Custom:Text
     Index:EFortEarnedSubRewardType Enum=@n1.RewardType
    ) -> (ReturnValue:Text)
   )
   > SetActiveWidget(
    self=Get Switcher_Icons() -> (Switcher_Icons:Common Visibility Switcher Object Reference)
    Widget=Select(
     XP:Overlay Object Reference=Get Overlay_XpiCON() -> (Overlay_XpiCON:Overlay Object Reference)
     Bars:Overlay Object Reference=Get Overlay_GoldIcon() -> (Overlay_GoldIcon:Overlay Object Reference)
     Custom:Overlay Object Reference
     Index=@n1.RewardType
    ) -> (ReturnValue:Overlay Object Reference)
   )
   > SetColorAndOpacity(
    self:Image Object Reference=Get Image_RarityColor_Glow() -> (Image_RarityColor_Glow:Image Object Reference)
    InColorAndOpacity:Linear Color Structure=Select(
     XP:Linear Color Structure=Get XPImageRarityGlowColor() -> (XPImageRarityGlowColor:Linear Color Structure)
     Bars:Linear Color Structure=Get BarsImageRarityGlowColor() -> (BarsImageRarityGlowColor:Linear Color Structure)
     Custom:Linear Color Structure
     Index=@n1.RewardType
    ) -> (ReturnValue:Linear Color Structure)
   )
   > SetActiveWidget(
    self:Widget Switcher Object Reference=Get Switcher_BG() -> (Switcher_BG:Widget Switcher Object Reference)
    Widget=Select(
     XP:Image Object Reference=Get ImageBG_XP() -> (ImageBG_XP:Image Object Reference)
     Bars:Image Object Reference=Get ImageBG_Bars() -> (ImageBG_Bars:Image Object Reference)
     Custom:Image Object Reference
     Index=@n1.RewardType
    ) -> (ReturnValue:Image Object Reference)
   )
   Event OnInitialized()
   > BindToAnimationFinished(Animation=Get Intro(), Delegate=Create Event())
   > BindToAnimationStarted!(execute, self:User Widget Object Reference, Animation:Widget Animation Object Reference=Get Intro(), Delegate:Delegate=Create Event()) -> (then)
  ]}]
variables [
 BarsImageBGColor {type=LinearColor, default=<rgba 0.8698 0.5631 0 1>}
 BarsImageBGColor2 {type=LinearColor, default=<rgba 0.2232 0.0648 0.0018 1>}
 BarsImageRarityGlowColor {type=LinearColor, default=<rgba 0.9922 0.6314 0 1>}
 XPImageBGColor {type=LinearColor, default=<rgba 0.1812 0.5089 0 1>}
 XPImageBGColor2 {type=LinearColor, default=<rgba 0.0648 0.1812 0 1>}
 XPImageRarityGlowColor {type=LinearColor, default=<rgba 0 0.8051 0.9922 1>}
]
widgets {
 info {inheritedWidgetCount=0, parentClass=/Script/FortniteUI.XpEarnedSubNotification, rootWidgetClass=&SizeBox, widgetCount=20}
 widgets [
  SizeBox_All {widgetClassPath=&SizeBox, bIsVariable=true, widget=SizeBox_All, properties {bOverride_MinDesiredWidth=true, minDesiredWidth=450}}
  VerticalBox {widgetClassPath=&VerticalBox, parent=SizeBox_All, widget=VerticalBox, slot {class=SizeBoxSlot}}
  SizeBox_LevelUP {
   widgetClassPath=&SizeBox
   bIsVariable=true
   parent=VerticalBox
   widget=SizeBox_LevelUP
   properties {bOverride_HeightOverride=true, heightOverride=49.3469, visibility=Collapsed}
   slot {class=VerticalBoxSlot, properties {padding=<ltrb 0 -8 0 0>}}
  }
  LevelUp_Header {
   widgetClassPath=&CommonBorder
   bIsVariable=true
   parent=SizeBox_LevelUP
   widget=LevelUp_Header
   properties {
    background {
     bIsDynamicallyLoaded=false
     drawAs=Box
     imageType=NoImage
     mirroring=NoMirror
     resourceName=None
     resourceObject=/FortUILibrary/Styles/Border/Materials/M_UI_AthenaObjectiveEntryHeader_Missions
     tiling=NoTile
     imageSize=<32 32>
     margin=<ltrb 0 0.5 0 0>
     outlineSettings {
      bUseBrushTransparency=false
      roundingType=HalfHeightRadius
      width=0
      color {colorUseRule=UseColor_Specified, specifiedColor=<rgba 0 0 0 0>}
      cornerRadii=<0 0 0 0>
     }
     tintColor {colorUseRule=UseColor_Specified, specifiedColor=<rgba 1 1 1 1>}
     uVRegion {bIsValid=false, max=<0 0>, min=<0 0>}
    }
    padding=<ltrb 16 14 16 0>
   }
   slot {class=SizeBoxSlot}
  }
  Text_RewardType {
   widgetClassPath=&CommonTextBlock
   bIsVariable=true
   parent=LevelUp_Header
   widget=Text_RewardType
   properties {
    text="XP EARNED!"
    textTransformPolicy=ToUpper
    font {
     bForceMonospaced=false
     bMaterialIsStencil=false
     fontMaterial=null
     fontObject=&Burbank
     letterSpacing=0
     monospacedWidth=1
     size=18
     skewAmount=0.2
     typefaceFontName=BurbankRegular-Black
     outlineSettings {
      bApplyOutlineToDropShadows=false
      bMiteredCorners=false
      bSeparateFillAlpha=false
      outlineMaterial=null
      outlineSize=0
      outlineColor=<rgba 0 0 0 1>
     }
    }
    shadowOffset=<0 0>
   }
   slot {class=BorderSlot, properties {padding=<ltrb 16 14 16 0>}}
  }
  SizeBox_2 {
   widgetClassPath=&SizeBox
   parent=VerticalBox
   widget=SizeBox_2
   properties {bOverride_HeightOverride=true, heightOverride=48}
   slot {class=VerticalBoxSlot}
  }
  Overlay-Reward {widgetClassPath=&Overlay, parent=SizeBox_2, widget=Overlay-Reward, slot {class=SizeBoxSlot}}
  Switcher_BG {
   widgetClassPath=&WidgetSwitcher
   bIsVariable=true
   parent=Overlay-Reward
   widget=Switcher_BG
   slot {class=OverlaySlot, properties {horizontalAlignment=HAlign_Fill, verticalAlignment=VAlign_Fill}}
  }
  ImageBG_XP {
   widgetClassPath=&Image
   bIsVariable=true
   parent=Switcher_BG
   widget=ImageBG_XP
   properties {
    visibility=SelfHitTestInvisible
    brush {
     bIsDynamicallyLoaded=false
     drawAs=Box
     imageType=FullColor
     mirroring=NoMirror
     resourceName=None
     resourceObject=/FortniteCoreUI/Athena/HUD/Quests/Art/M_Ui_Quest_XPNotification_DarkPurple
     tiling=NoTile
     imageSize=<64 64>
     margin=<ltrb 0.3 0.3 0.3 0.3>
     outlineSettings {
      bUseBrushTransparency=false
      roundingType=HalfHeightRadius
      width=0
      color {colorUseRule=UseColor_Specified, specifiedColor=<rgba 0 0 0 0>}
      cornerRadii=<0 0 0 0>
     }
     tintColor {colorUseRule=UseColor_Specified, specifiedColor=<rgba 1 1 1 1>}
     uVRegion {bIsValid=false, max=<0 0>, min=<0 0>}
    }
   }
   slot {class=WidgetSwitcherSlot}
  }
  ImageBG_Bars {
   widgetClassPath=&Image
   bIsVariable=true
   parent=Switcher_BG
   widget=ImageBG_Bars
   properties {
    visibility=SelfHitTestInvisible
    brush {
     bIsDynamicallyLoaded=false
     drawAs=Box
     imageType=NoImage
     mirroring=NoMirror
     resourceName=None
     resourceObject=/FortniteCoreUI/Athena/HUD/Quests/Art/M_Ui_Quest_XPNotification_Bars
     tiling=NoTile
     imageSize=<64 64>
     margin=<ltrb 0.3 0.3 0.3 0.3>
     outlineSettings {
      bUseBrushTransparency=false
      roundingType=HalfHeightRadius
      width=0
      color {colorUseRule=UseColor_Specified, specifiedColor=<rgba 0 0 0 0>}
      cornerRadii=<0 0 0 0>
     }
     tintColor {colorUseRule=UseColor_Specified, specifiedColor=<rgba 1 1 1 1>}
     uVRegion {bIsValid=false, max=<0 0>, min=<0 0>}
    }
   }
   slot {class=WidgetSwitcherSlot}
  }
  Overlay_4 {
   widgetClassPath=&Overlay
   parent=Overlay-Reward
   widget=Overlay_4
   slot {class=OverlaySlot, properties {horizontalAlignment=HAlign_Fill, verticalAlignment=VAlign_Fill}}
  }
  Image_RarityColor_Glow {
   widgetClassPath=&Image
   bIsVariable=true
   parent=Overlay_4
   widget=Image_RarityColor_Glow
   properties {
    brush {
     bIsDynamicallyLoaded=false
     drawAs=Image
     imageType=NoImage
     mirroring=NoMirror
     resourceName=None
     resourceObject=/Game/Athena/HUD/LTM/Wax/Materials/M_UI_TextGlow_Inst
     tiling=NoTile
     imageSize=<44 44>
     margin=<ltrb 0 0 0 0>
     outlineSettings {
      bUseBrushTransparency=false
      roundingType=HalfHeightRadius
      width=0
      color {colorUseRule=UseColor_Specified, specifiedColor=<rgba 0 0 0 0>}
      cornerRadii=<0 0 0 0>
     }
     tintColor {colorUseRule=UseColor_Specified, specifiedColor=<rgba 1 1 1 1>}
     uVRegion {bIsValid=false, max=<0 0>, min=<0 0>}
    }
    colorAndOpacity=<rgba 0 0.8051 0.9922 0.5>
    renderTransform {angle=0, scale=<2 2>, shear=<0 0>, translation=<0 0>}
    renderTransformPivot=<0.75 0.5>
   }
   slot {class=OverlaySlot, properties {horizontalAlignment=HAlign_Right, verticalAlignment=VAlign_Center}}
  }
  ScaleBox_2 {
   widgetClassPath=&ScaleBox
   parent=Overlay_4
   widget=ScaleBox_2
   properties {stretchDirection=DownOnly}
   slot {class=OverlaySlot, properties {horizontalAlignment=HAlign_Fill, verticalAlignment=VAlign_Center, padding=<ltrb 8 0 85 0>}}
  }
  Text_XpValue {
   widgetClassPath=&CommonTextBlock
   bIsVariable=true
   parent=ScaleBox_2
   widget=Text_XpValue
   properties {
    text="+ 00,000 XP!"
    textTransformPolicy=ToUpper
    font {
     bForceMonospaced=false
     bMaterialIsStencil=false
     fontMaterial=null
     fontObject=&HeadingNow
     letterSpacing=0
     monospacedWidth=1
     size=18.75
     skewAmount=0
     typefaceFontName=86 Bold
     outlineSettings {
      bApplyOutlineToDropShadows=false
      bMiteredCorners=false
      bSeparateFillAlpha=false
      outlineMaterial=null
      outlineSize=0
      outlineColor=<rgba 0 0 0 1>
     }
    }
    renderTransform {angle=0, scale=<1 1>, shear=<-12 0>, translation=<0 0>}
    shadowColorAndOpacity=<rgba 0 0 0 0.15>
    shadowOffset=<5 3>
   }
   slot {class=ScaleBoxSlot, properties {horizontalAlignment=HAlign_Right}}
  }
  Switcher_Icons {
   widgetClassPath=&CommonVisibilitySwitcher
   bIsVariable=true
   parent=Overlay_4
   widget=Switcher_Icons
   slot {class=OverlaySlot, properties {horizontalAlignment=HAlign_Right}}
  }
  Overlay_XpiCON {
   widgetClassPath=&Overlay
   bIsVariable=true
   parent=Switcher_Icons
   widget=Overlay_XpiCON
   properties {renderTransform {angle=0, scale=<1.2 1.2>, shear=<0 0>, translation=<-5 10>}, renderTransformPivot=<1 1>}
   slot {class=CommonVisibilitySwitcherSlot}
  }
  Image_Icon {
   widgetClassPath=&Image
   bIsVariable=true
   parent=Overlay_XpiCON
   widget=Image_Icon
   properties {
    brush {
     bIsDynamicallyLoaded=false
     drawAs=Image
     imageType=FullColor
     mirroring=NoMirror
     resourceName=None
     resourceObject=/Game/UI/Foundation/Textures/Icons/Items/UniversalXP/T_UI_UniversalXP_Icon_Standard_128
     tiling=NoTile
     imageSize=<30 30>
     margin=<ltrb 0 0 0 0>
     outlineSettings {
      bUseBrushTransparency=false
      roundingType=HalfHeightRadius
      width=0
      color {colorUseRule=UseColor_Specified, specifiedColor=<rgba 0 0 0 0>}
      cornerRadii=<0 0 0 0>
     }
     tintColor {colorUseRule=UseColor_Specified, specifiedColor=<rgba 1 1 1 1>}
     uVRegion {bIsValid=false, max=<0 0>, min=<0 0>}
    }
    renderTransform {angle=0, scale=<1.25 1.25>, shear=<0 0>, translation=<0 0>}
   }
   slot {class=OverlaySlot, properties {horizontalAlignment=HAlign_Center, verticalAlignment=VAlign_Center, padding=<ltrb 0 2 4 0>}}
  }
  Overlay_GoldIcon {
   widgetClassPath=&Overlay
   bIsVariable=true
   parent=Switcher_Icons
   widget=Overlay_GoldIcon
   properties {renderTransform {angle=0, scale=<1.6 1.6>, shear=<0 0>, translation=<0 12>}, renderTransformPivot=<1 1>}
   slot {class=CommonVisibilitySwitcherSlot}
  }
  Image_shadow_1 {
   widgetClassPath=&Image
   bIsVariable=true
   parent=Overlay_GoldIcon
   widget=Image_shadow_1
   properties {
    brush {
     bIsDynamicallyLoaded=false
     drawAs=Image
     imageType=NoImage
     mirroring=NoMirror
     resourceName=None
     resourceObject=/Game/UI/Foundation/Textures/Icons/Athena/T-T-Icon-BR-GoldBars-UI-Icon
     tiling=NoTile
     imageSize=<48 48>
     margin=<ltrb 0 0 0 0>
     outlineSettings {
      bUseBrushTransparency=false
      roundingType=HalfHeightRadius
      width=0
      color {colorUseRule=UseColor_Specified, specifiedColor=<rgba 0 0 0 0>}
      cornerRadii=<0 0 0 0>
     }
     tintColor {colorUseRule=UseColor_Specified, specifiedColor=<rgba 0 0 0 1>}
     uVRegion {bIsValid=false, max=<0 0>, min=<0 0>}
    }
    colorAndOpacity=<rgba 1 1 1 0.25>
    renderTransform {angle=0, scale=<1 1>, shear=<-2 0>, translation=<2 1>}
   }
   slot {class=OverlaySlot}
  }
  Image_Icon_1 {
   widgetClassPath=&Image
   bIsVariable=true
   parent=Overlay_GoldIcon
   widget=Image_Icon_1
   properties {brush {
     bIsDynamicallyLoaded=false
     drawAs=Image
     imageType=NoImage
     mirroring=NoMirror
     resourceName=None
     resourceObject=&T-T-Icon-BR-GoldBars-UI-Icon
     tiling=NoTile
     imageSize=<48 48>
     margin=<ltrb 0 0 0 0>
     outlineSettings {
      bUseBrushTransparency=false
      roundingType=HalfHeightRadius
      width=0
      color {colorUseRule=UseColor_Specified, specifiedColor=<rgba 0 0 0 0>}
      cornerRadii=<0 0 0 0>
     }
     tintColor {colorUseRule=UseColor_Specified, specifiedColor=<rgba 1 1 1 1>}
     uVRegion {bIsValid=false, max=<0 0>, min=<0 0>}
    }}
   slot {class=OverlaySlot}
  }
 ]
}
