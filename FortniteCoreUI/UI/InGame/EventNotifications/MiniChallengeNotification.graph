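# Event notification widget for mini challenges, built on FortNotificationEntry_MiniChallenge, showing a MiniChallengeEntry on a dark translucent panel above an XpSubNotification.
# Construct binds the entry's toast delegates and the intro animation start, which appears to drive the popup sound.
# Intro slides and fades in while calling the entry's CompetionToast and finishing the intro phase; outro fades out and reports completion.
# The XP toast handler is empty, and a delayed FadeToastAway chain has no triggering event.
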WidgetBlueprint /FortniteCoreUI/UI/InGame/EventNotifications/MiniChallengeNotification
animations [
 Outro {
  binding_count=1
  length_seconds=0.25
  object=&self:Outro
  sequence {bindings [VerticalBox_0 {
     id=#1
     possessed_class=VerticalBox
     tracks [{
       type=MovieSceneFloatTrack
       display_name=Render Opacity
       sections [{type=MovieSceneFloatSection, end_frame=5, start_frame=0, channels [None {type=MovieSceneScriptingFloatChannel, keys [[0,1],[5,0]]}]}]
      }]
    }]}
 }
 intro {
  binding_count=1
  length_seconds=0.5
  object=&self:intro
  sequence {bindings [VerticalBox_0 {
     id=#2
     possessed_class=VerticalBox
     tracks [
      {
       type=MovieSceneFloatTrack
       display_name=Render Opacity
       sections [{type=MovieSceneFloatSection, end_frame=10, start_frame=0, channels [None {type=MovieSceneScriptingFloatChannel, keys [[0,0],[10,1]]}]}]
      }
      {
       type=MovieScene2DTransformTrack
       sections [{
         type=MovieScene2DTransformSection
         end_frame=10
         start_frame=0
         channels [
          Translation.X {type=MovieSceneScriptingFloatChannel, keys [[0,-100],[10,0]]}
          Translation.Y {type=MovieSceneScriptingFloatChannel}
          Angle {type=MovieSceneScriptingFloatChannel}
          Scale.X {type=MovieSceneScriptingFloatChannel}
          Scale.Y {type=MovieSceneScriptingFloatChannel}
          Shear.X {type=MovieSceneScriptingFloatChannel}
          Shear.Y {type=MovieSceneScriptingFloatChannel}
         ]
        }]
      }
     ]
    }]}
 }
]
graphs [EventGraph {nodes [
   Play Inro Sound() {is=title K2Node_CustomEvent}
   > PlaySound2D(Sound=/Game/Athena/Sounds/Quests/MiniChallenges/Notification_Popup_Cue)
   n1 = MiniChallengeToastFinished() {is=title K2Node_CustomEvent}
   FinishNotificationIntro(execute=[@n1.then, @n4.CastFailed], self:Fort Notification Entry Object Reference)
   n2 = Cast To MiniChallengeEntry!(execute, Object=Get MiniChallenge() -> (MiniChallenge:Athena Mini Challenge Entry Object Reference)) -> (then, CastFailed, AsMini Challenge Entry:Mini Challenge Entry Object Reference) {is=title K2Node_DynamicCast}
   > n3 = FadeToastAway(self:Mini Challenge Entry Object Reference=@n2)
   Delay(execute=[@n3.then, @n2.CastFailed], Duration=1)
   n4 = Cast To MiniChallengeEntry(Object=Get MiniChallenge())
   > CompetionToast(self:Mini Challenge Entry Object Reference=@n4)
   Event Construct()
   > n5 = Cast To MiniChallengeEntry(Object=Get MiniChallenge())
   > Assign Toast Finished!(execute:Exec, self:Mini Challenge Entry Object Reference=@n5, Delegate:"Delegate (by ref)"=Create Event()) -> (then:Exec) {is=title K2Node_AssignDelegate}
   > Assign Trigger Xp Toast!(execute:Exec, self:Mini Challenge Entry Object Reference=@n5, Delegate:"Delegate (by ref)"=Create Event()) -> (then:Exec) {is=title K2Node_AssignDelegate}
   > BindToAnimationStarted!(execute, self:User Widget Object Reference, Animation:Widget Animation Object Reference=Get Intro(), Delegate:Delegate=Create Event()) -> (then)
   Event StartNotificationIntro() {is=title K2Node_Event}
   > PlayAnimationForward(InAnimation=Get Intro())
   > n6 = Cast To MiniChallengeEntry(Object=Get MiniChallenge())
   > n7 = CompetionToast(self=@n6)
   FinishNotificationIntro(execute=[@n7.then, @n6.CastFailed])
   Event StartNotificationOutro() {is=title K2Node_Event}
   > PlayAnimationForward(InAnimation=Get Outro() -> (Outro:Widget Animation Object Reference))
   "Animation Finished (Outro)"!() -> (OutputDelegate:Delegate, then:Exec) {is=title K2Node_WidgetAnimationEvent}
   > FinishNotificationOutro(self:Fort Notification Entry Object Reference)
   TriggerXpToast_Event_0() {is=title K2Node_CustomEvent}
  ]}]
variables []
widgets {
 info {
  inheritedWidgetCount=2
  parentClass=/Script/FortniteUI.FortNotificationEntry_MiniChallenge
  rootWidgetClass=&VerticalBox
  widgetCount=5
 }
 widgets [
  VerticalBox_0 {widgetClassPath=&VerticalBox, widget=VerticalBox_0}
  SizeBox_0 {
   widgetClassPath=&SizeBox
   parent=VerticalBox_0
   widget=SizeBox_0
   properties {bOverride_WidthOverride=true, widthOverride=450}
   slot {class=VerticalBoxSlot}
  }
  Border_0 {
   widgetClassPath=&Border
   parent=SizeBox_0
   widget=Border_0
   properties {
    visibility=SelfHitTestInvisible
    background {
     bIsDynamicallyLoaded=false
     drawAs=Image
     imageType=NoImage
     mirroring=NoMirror
     resourceName=None
     resourceObject=null
     tiling=NoTile
     imageSize=<32 32>
     margin=<ltrb 0 0 0 0>
     outlineSettings {
      bUseBrushTransparency=false
      roundingType=HalfHeightRadius
      width=0
      color {colorUseRule=UseColor_Specified, specifiedColor=<rgba 0 0 0 0>}
      cornerRadii=<0 0 0 0>
     }
     tintColor {colorUseRule=UseColor_Specified, specifiedColor=<rgba 0.0052 0.0052 0.0056 0.498>}
     uVRegion {bIsValid=false, max=<0 0>, min=<0 0>}
    }
    padding=<ltrb 4 4 4 4>
   }
   slot {class=SizeBoxSlot}
  }
  MiniChallenge {
   widgetClassPath=/FortniteCoreUI/Athena/UI/Challenges/MiniChallengeEntry
   bInherited=true
   bIsVariable=true
   parent=Border_0
   widget=MiniChallenge
   properties {notificationDisplay=true}
   slot {class=BorderSlot, properties {padding=<ltrb 4 4 4 4>}}
  }
  XpSubNotification {
   widgetClassPath=&/XpSubNotification
   bInherited=true
   bIsVariable=true
   parent=VerticalBox_0
   widget=XpSubNotification
   slot {class=VerticalBoxSlot}
  }
 ]
}
