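# Reward tile widget for received-item event notifications, showing an item image, a rarity-style glow and a stack count.
# UpdateVisuals collapses the tile, sets the count text and calls RarityVisuals, which tints the flare material but looks up Mythic colors regardless of the rarity passed in.
# ShowReward reveals the tile and plays the unlock and quantity animations.
# PreConstruct previews the glow at design time using DevRarity.
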WidgetBlueprint /FortniteCoreUI/UI/InGame/EventNotifications/FreshCheeseDisplayItem
animations [
 RewardUnlocked {
  binding_count=8
  length_seconds=1.75
  object=&self:RewardUnlocked
  sequence {bindings [
    Backing {
     id=#1
     display_name=FullWidget
     possessed_class=CommonBorder
     tracks [
      {
       type=MovieSceneFloatTrack
       display_name=Render Opacity
       sections [{type=MovieSceneFloatSection, channels [None {type=MovieSceneScriptingFloatChannel, keys [[1,0],[6,1],[35,1]]}]}]
      }
      {
       type=MovieScene2DTransformTrack
       sections [{
         type=MovieScene2DTransformSection
         channels [
          Translation.X {type=MovieSceneScriptingFloatChannel, keys [[1,-64],[6,0]]}
          Translation.Y {type=MovieSceneScriptingFloatChannel}
          Angle {type=MovieSceneScriptingFloatChannel}
          Scale.X {type=MovieSceneScriptingFloatChannel}
          Scale.Y {type=MovieSceneScriptingFloatChannel}
          Shear.X {type=MovieSceneScriptingFloatChannel}
          Shear.Y {type=MovieSceneScriptingFloatChannel}
         ]
        }]
      }
     ]
    }
    Reward_Preview {
     id=#2
     possessed_class=FortLazyImage
     tracks [
      {
       type=MovieSceneFloatTrack
       display_name=Render Opacity
       sections [{type=MovieSceneFloatSection, channels [None {type=MovieSceneScriptingFloatChannel, keys [[4,0],[6,1]]}]}]
      }
      {
       type=MovieScene2DTransformTrack
       sections [{
         type=MovieScene2DTransformSection
         channels [
          Translation.X {type=MovieSceneScriptingFloatChannel}
          Translation.Y {type=MovieSceneScriptingFloatChannel}
          Angle {type=MovieSceneScriptingFloatChannel}
          Scale.X {type=MovieSceneScriptingFloatChannel, keys [[4,1.5],[6,1],[7,1]]}
          Scale.Y {type=MovieSceneScriptingFloatChannel, keys [[4,1.5],[6,1],[7,1]]}
          Shear.X {type=MovieSceneScriptingFloatChannel}
          Shear.Y {type=MovieSceneScriptingFloatChannel}
         ]
        }]
      }
     ]
    }
    RewardFlare {
     id=#3
     display_name=Reward_Glow
     possessed_class=Image
     tracks [
      {
       type=MovieSceneFloatTrack
       display_name=Render Opacity
       sections [{type=MovieSceneFloatSection, channels [None {type=MovieSceneScriptingFloatChannel, keys [[4,0],[6,0.5],[6.0003,1],[35,0.25]]}]}]
      }
      {
       type=MovieSceneWidgetMaterialTrack
       sections [{
         type=MovieSceneParameterSection
         channels [
          FlareDensity {type=MovieSceneScriptingFloatChannel, keys [[4,1.5],[7,2],[35,2]]}
          FlareRadius {type=MovieSceneScriptingFloatChannel, keys [[4,0.75],[7,1.175]]}
         ]
        }]
      }
     ]
    }
    Overlay_5 {
     id=#4
     display_name=Reward_Preview_Shadow
     possessed_class=Overlay
     tracks [{
       type=MovieSceneFloatTrack
       display_name=Render Opacity
       sections [{type=MovieSceneFloatSection, start_frame=0, channels [None {type=MovieSceneScriptingFloatChannel, keys [[0,0],[1,1]]}]}]
      }]
    }
   ]}
 }
 Intro {
  binding_count=1
  length_seconds=0.5
  object=&self:Intro
  sequence {bindings [SizeBox_1 {
     id=#5
     display_name=FullWidget
     possessed_class=SizeBox
     tracks [
      {
       type=MovieSceneFloatTrack
       display_name=Render Opacity
       sections [{type=MovieSceneFloatSection, channels [None {type=MovieSceneScriptingFloatChannel, keys [[0,0],[5,1]]}]}]
      }
      {
       type=MovieScene2DTransformTrack
       sections [{
         type=MovieScene2DTransformSection
         channels [
          Translation.X {type=MovieSceneScriptingFloatChannel, keys [[0,-50],[10,0]]}
          Translation.Y {type=MovieSceneScriptingFloatChannel}
          Angle {type=MovieSceneScriptingFloatChannel}
          Scale.X {type=MovieSceneScriptingFloatChannel}
          Scale.Y {type=MovieSceneScriptingFloatChannel}
          Shear.X {type=MovieSceneScriptingFloatChannel}
          Shear.Y {type=MovieSceneScriptingFloatChannel}
         ]
        }]
      }
     ]
    }]}
 }
 Outro {
  binding_count=1
  length_seconds=0.25
  object=&self:Outro
  sequence {bindings [SizeBox_1 {
     id=#6
     display_name=FullWidget
     possessed_class=SizeBox
     tracks [{
       type=MovieSceneFloatTrack
       display_name=Render Opacity
       sections [{type=MovieSceneFloatSection, channels [None {type=MovieSceneScriptingFloatChannel, keys [[0,1],[5,0]]}]}]
      }]
    }]}
 }
 QuantityAnim {
  binding_count=1
  length_seconds=0.9001
  object=&self:QuantityAnim
  sequence {bindings [Count_Hbox {
     id=#7
     possessed_class=HorizontalBox
     tracks [
      {
       type=MovieScene2DTransformTrack
       sections [{
         type=MovieScene2DTransformSection
         end_frame=18
         start_frame=0
         channels [
          Translation.X {type=MovieSceneScriptingFloatChannel, const [-5,3]}
          Translation.Y {type=MovieSceneScriptingFloatChannel, const [0,3]}
          Angle {type=MovieSceneScriptingFloatChannel}
          Scale.X {type=MovieSceneScriptingFloatChannel, keys [[0,3],[8.0003,3],[14.0007,0.9],[16.0007,1.1],[18.001,1]]}
          Scale.Y {type=MovieSceneScriptingFloatChannel, keys [[0,3],[8.0003,3],[14.0007,0.9],[16.0007,1.1],[18.001,1]]}
          Shear.X {type=MovieSceneScriptingFloatChannel, const [0,3]}
          Shear.Y {type=MovieSceneScriptingFloatChannel, const [0,3]}
         ]
        }]
      }
      {
       type=MovieSceneFloatTrack
       display_name=Render Opacity
       sections [{
         type=MovieSceneFloatSection
         end_frame=18
         start_frame=0
         channels [None {type=MovieSceneScriptingFloatChannel, thinned_from=4, keys [[0,0],[8.0003,0],[18.001,1]]}]
        }]
      }
     ]
    }]}
 }
]
graphs [
 EventGraph {nodes [
   RarityVisuals!() -> (OutputDelegate, then, Rarity:EFortRarity Enum=Common) {is=title K2Node_CustomEvent}
   > n1 = GetRarityData!(execute, Rarity:EFortRarity Enum=Mythic) -> (then, ReturnValue:Fort Rarity Item Data Structure)
   > n2 = GetDynamicMaterial(self=Get Reward_Glow() -> (Reward_Glow:Image Object Reference))
   n3 = Break Fort Rarity Item Data!(FortRarityItemData:"Fort Rarity Item Data Structure (by ref)"=@n1) -> (
    Name:Text
    Color1:Linear Color Structure
    Color2:Linear Color Structure
    Color3:Linear Color Structure
    Color4:Linear Color Structure
    Color5:Linear Color Structure
    Radius:float=~
    Falloff:float=~
    Brightness:float=~
    Roughness:float=~
    Glow:float=~
   ) {is=title K2Node_BreakStruct}
   > SetVectorParameterValue(self=@n2, ParameterName=RGB 2, Value="Lerp (LinearColor)"(A=@n3.Color1, B=(R=1,G=1,B=1,A=1), Alpha=0.6))
   > SetVectorParameterValue(self=@n2, ParameterName=RGB 1, Value=@n3.Color3)
   n4 = Event PreConstruct()
   > Branch(Condition=@n4.IsDesignTime)
   > RarityVisuals!(execute, self, Rarity:EFortRarity Enum=Get DevRarity() -> (DevRarity:EFortRarity Enum=Common)) -> (then) {is=title K2Node_CallFunction}
   n5 = Event UpdateVisuals!() -> (OutputDelegate, then, RewardItem:Item Definition Base Object Reference, ItemQuantity:int=0) {is=title K2Node_Event}
   > SetVisibility(InVisibility=Collapsed)
   n6 = IsValid(Object=@n5.RewardItem)
   > SetVisibility(
    self=Get FullWidget() -> (FullWidget:Overlay Object Reference)
    InVisibility=Select(Option 0=Collapsed, Option 1=SelfHitTestInvisible, Index=@n6)
   )
   > Branch(Condition=@n6)
   > RarityVisuals(Rarity=GetRarity!(ItemDefinition:Item Definition Base Object Reference=@n5.RewardItem) -> (ReturnValue:EFortRarity Enum=Common))
   > SetText(
    self=Get RichText_StackCount() -> (RichText_StackCount:Rich Text Block Object Reference)
    InText=Format Text!(Format="LOCTEXT(\"<countGrey>x</><countNumber>{n}</>\")", n:int=@n5.ItemQuantity) -> (Result)
   )
  ]}
 ShowReward {nodes [
   ShowReward() {is=title K2Node_FunctionEntry}
   > SetVisibility(InVisibility=SelfHitTestInvisible)
   > PlayAnimation(InAnimation=Get RewardUnlocked() -> (RewardUnlocked:Widget Animation Object Reference))
   > PlayAnimation(InAnimation=Get QuantityAnim() -> (QuantityAnim:Widget Animation Object Reference))
  ]}
]
variables [
 DevRarity {type=EFortRarity, default=Epic}
 IntroSpeed {type=number, default=3}
]
widgets {
 info {inheritedWidgetCount=2, parentClass=/Script/FortniteUI.FortMultipleReceivedItemWidget, rootWidgetClass=&Overlay, widgetCount=6}
 widgets [
  FullWidget {widgetClassPath=&Overlay, bIsVariable=true, widget=FullWidget}
  Reward_Glow {
   widgetClassPath=&Image
   bIsVariable=true
   parent=FullWidget
   widget=Reward_Glow
   properties {
    renderOpacity=0.7
    brush {
     bIsDynamicallyLoaded=false
     drawAs=Image
     imageType=NoImage
     mirroring=NoMirror
     resourceName=None
     resourceObject=/Game/UI/Foundation/Onboarding/Materials/M_FX_Flare_Master_Toast
     tiling=NoTile
     imageSize=<96 96>
     margin=<ltrb 0 0 0 0>
     outlineSettings {
      bUseBrushTransparency=false
      roundingType=HalfHeightRadius
      width=0
      color {colorUseRule=UseColor_Specified, specifiedColor=<rgba 0 0 0 0>}
      cornerRadii=<0 0 0 0>
     }
     tintColor {colorUseRule=UseColor_Specified, specifiedColor=<rgba 1 1 1 1>}
     uVRegion {bIsValid=false, max=<0 0>, min=<0 0>}
    }
    renderTransform {angle=0, scale=<1.7 1.7>, shear=<0 0>, translation=<0 0>}
   }
   slot {class=OverlaySlot, properties {horizontalAlignment=HAlign_Center, verticalAlignment=VAlign_Center}}
  }
  Reward_Preview_Shadow {
   widgetClassPath=&FortLazyImage
   bInherited=true
   bIsVariable=true
   parent=FullWidget
   widget=Reward_Preview_Shadow
   properties {
    designerPreviewItem=/Game/Items/PersistentResources/AthenaBattleStar
    renderOpacity=0
    brush {
     bIsDynamicallyLoaded=false
     drawAs=Image
     imageType=NoImage
     mirroring=NoMirror
     resourceName=None
     resourceObject=/Game/Athena/UI/Frontend/Art/T_UI_BP_BattleStar_64
     tiling=NoTile
     imageSize=<88 88>
     margin=<ltrb 0 0 0 0>
     outlineSettings {
      bUseBrushTransparency=false
      roundingType=HalfHeightRadius
      width=0
      color {colorUseRule=UseColor_Specified, specifiedColor=<rgba 0 0 0 0>}
      cornerRadii=<0 0 0 0>
     }
     tintColor {colorUseRule=UseColor_Specified, specifiedColor=<rgba 0 0 0 1>}
     uVRegion {bIsValid=false, max=<0 0>, min=<0 0>}
    }
    colorAndOpacity=<rgba 1 1 1 0.35>
   }
   slot {class=OverlaySlot, properties {horizontalAlignment=HAlign_Center, verticalAlignment=VAlign_Center}}
  }
  Reward_Preview {
   widgetClassPath=&FortLazyImage
   bInherited=true
   bIsVariable=true
   parent=FullWidget
   widget=Reward_Preview
   properties {
    designerPreviewItem=&AthenaBattleStar
    brush {
     bIsDynamicallyLoaded=false
     drawAs=Image
     imageType=NoImage
     mirroring=NoMirror
     resourceName=None
     resourceObject=&T_UI_BP_BattleStar_64
     tiling=NoTile
     imageSize=<88 88>
     margin=<ltrb 0 0 0 0>
     outlineSettings {
      bUseBrushTransparency=false
      roundingType=HalfHeightRadius
      width=0
      color {colorUseRule=UseColor_Specified, specifiedColor=<rgba 0 0 0 0>}
      cornerRadii=<0 0 0 0>
     }
     tintColor {colorUseRule=UseColor_Specified, specifiedColor=<rgba 1 1 1 1>}
     uVRegion {bIsValid=false, max=<0 0>, min=<0 0>}
    }
    loadingBackgroundBrush {
     bIsDynamicallyLoaded=false
     drawAs=Image
     imageType=NoImage
     mirroring=NoMirror
     resourceName=None
     resourceObject=/FortUILibrary/UIKit/Throbber/MI_UIKit_Monochrome_Throbber
     tiling=NoTile
     imageSize=<32 32>
     margin=<ltrb 0 0 0 0>
     outlineSettings {
      bUseBrushTransparency=false
      roundingType=HalfHeightRadius
      width=0
      color {colorUseRule=UseColor_Specified, specifiedColor=<rgba 0 0 0 0>}
      cornerRadii=<0 0 0 0>
     }
     tintColor {colorUseRule=UseColor_Specified, specifiedColor=<rgba 1 1 1 1>}
     uVRegion {bIsValid=false, max=<0 0>, min=<0 0>}
    }
   }
   slot {class=OverlaySlot, properties {horizontalAlignment=HAlign_Center, verticalAlignment=VAlign_Center}}
  }
  Count_Hbox {
   widgetClassPath=&HorizontalBox
   bIsVariable=true
   parent=FullWidget
   widget=Count_Hbox
   properties {renderTransform {angle=-10, scale=<1 1>, shear=<0 0>, translation=<-5 0>}}
   slot {class=OverlaySlot, properties {horizontalAlignment=HAlign_Right, verticalAlignment=VAlign_Bottom}}
  }
  RichText_StackCount {
   widgetClassPath=/Script/UMG.RichTextBlock
   bIsVariable=true
   parent=Count_Hbox
   widget=RichText_StackCount
   properties {text="<countGrey>x</><countNumber>[n]</>", textStyleSet=/Game/UI/InGame/EventNotifications/MultipleReceivedItem_TextStyles}
   slot {class=HorizontalBoxSlot, properties {verticalAlignment=VAlign_Bottom}}
  }
 ]
}
