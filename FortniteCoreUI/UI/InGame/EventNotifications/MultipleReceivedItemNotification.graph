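# HUD notification widget that shows a three-item "PERSONAL BONUS!" reward, or a squad-up-with-friends reminder when no bonus will be received.
# StartNotificationIntro animates the header and picks a switcher page, using timers to reveal items in sequence or hold the reminder before finishing the intro.
# Each reveal recolors the backing and flare dynamic materials from that item's DevRarity; the reminder uses Common.
# The outro fades the header, then signals completion.
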WidgetBlueprint /FortniteCoreUI/UI/InGame/EventNotifications/MultipleReceivedItemNotification
animations [
 Intro {
  binding_count=1
  length_seconds=0.5
  object=&self:Intro
  sequence {bindings [SizeBox_1 {
     id=#1
     display_name=Overlay_5
     possessed_class=SizeBox
     tracks [
      {
       type=MovieSceneFloatTrack
       display_name=Render Opacity
       sections [{type=MovieSceneFloatSection, channels [None {type=MovieSceneScriptingFloatChannel, keys [[0,0],[5,1]]}]}]
      }
      {
       type=MovieScene2DTransformTrack
       sections [{
         type=MovieScene2DTransformSection
         channels [
          Translation.X {type=MovieSceneScriptingFloatChannel, keys [[0,-50],[10,0]]}
          Translation.Y {type=MovieSceneScriptingFloatChannel}
          Angle {type=MovieSceneScriptingFloatChannel}
          Scale.X {type=MovieSceneScriptingFloatChannel}
          Scale.Y {type=MovieSceneScriptingFloatChannel}
          Shear.X {type=MovieSceneScriptingFloatChannel}
          Shear.Y {type=MovieSceneScriptingFloatChannel}
         ]
        }]
      }
     ]
    }]}
 }
 Outro {
  binding_count=1
  length_seconds=0.25
  object=&self:Outro
  sequence {bindings [SizeBox_1 {
     id=#2
     display_name=Overlay_5
     possessed_class=SizeBox
     tracks [{
       type=MovieSceneFloatTrack
       display_name=Render Opacity
       sections [{type=MovieSceneFloatSection, channels [None {type=MovieSceneScriptingFloatChannel, keys [[0,1],[5,0]]}]}]
      }]
    }]}
 }
]
graphs [
 EventGraph {nodes [
   n1 = HandleRewardEnd() {is=title K2Node_CustomEvent}
   > FinishNotificationIntro(self:Fort Notification Entry Object Reference)
   n2 = RarityVisuals!() -> (OutputDelegate, then, Rarity:EFortRarity Enum=Common) {is=title K2Node_CustomEvent}
   > n3 = GetRarityData!(execute, Rarity:EFortRarity Enum=@n2.Rarity) -> (then, ReturnValue:Fort Rarity Item Data Structure)
   > n4 = GetDynamicMaterial(self:Border Object Reference=Get Backing() -> (Backing:Common Border Object Reference))
   n5 = Break Fort Rarity Item Data!(FortRarityItemData:"Fort Rarity Item Data Structure (by ref)"=@n3) -> (
    Name:Text
    Color1:Linear Color Structure
    Color2:Linear Color Structure
    Color3:Linear Color Structure
    Color4:Linear Color Structure
    Color5:Linear Color Structure
    Radius:float=~
    Falloff:float=~
    Brightness:float=~
    Roughness:float=~
    Glow:float=~
   ) {is=title K2Node_BreakStruct}
   > SetVectorParameterValue(self=@n4, ParameterName=RGB-Rarity1, Value=@n5.Color2)
   > SetVectorParameterValue(self=@n4, ParameterName=Background_color, Value=@n5.Color4)
   > SetVectorParameterValue(self=@n4, ParameterName=Triangle_color, Value=@n5.Color3)
   > n6 = GetDynamicMaterial(self:Image Object Reference=Get RewardFlare() -> (RewardFlare:Image Object Reference))
   > SetVectorParameterValue(self=@n6, ParameterName=Triangle_RGB_B, Value="Lerp (LinearColor)"(A=@n5.Color1, B=(R=1,G=1,B=1,A=1), Alpha=0.6))
   > SetVectorParameterValue(self=@n6, ParameterName=RGB_InnerFlame, Value=@n5.Color1)
   > SetVectorParameterValue(self=@n6, ParameterName=Triangle_RGB_A, Value=@n5.Color1)
   > SetVectorParameterValue(self=@n6, ParameterName=Triangle_RGB_A, Value=@n5.Color2)
   n7 = HandleReward_2() {is=title K2Node_CustomEvent}
   > n8 = Cast To FreshCheeseDisplayItem!(execute, Object=Get DisplayItem_2() -> (DisplayItem_2:Fort Multiple Received Item Widget Object Reference)) -> (then, CastFailed, AsFresh Cheese Display Item:Fresh Cheese Display Item Object Reference) {is=title K2Node_DynamicCast}
   > ShowReward(self:Fresh Cheese Display Item Object Reference=@n8)
   > RarityVisuals!(
    execute
    self
    Rarity:EFortRarity Enum=Get DevRarity!(self:Fresh Cheese Display Item Object Reference=@n8) -> (DevRarity:EFortRarity Enum=Common)
   ) -> (then) {is=title K2Node_CallFunction}
   n9 = HandleReward_3() {is=title K2Node_CustomEvent}
   > n10 = Cast To FreshCheeseDisplayItem(Object=Get DisplayItem_3() -> (DisplayItem_3:Fort Multiple Received Item Widget Object Reference))
   > ShowReward(self=@n10)
   > RarityVisuals(Rarity=Get DevRarity(self=@n10) -> (DevRarity=Common))
   n11 = HandleNoFriendsMessage() {is=title K2Node_CustomEvent}
   > FinishNotificationIntro()
   n12 = Event PreConstruct()
   > Branch(Condition=@n12.IsDesignTime)
   > ShowNoFriendsMessage()
   Event StartNotificationIntro() {is=title K2Node_Event}
   > PlayAnimation(InAnimation=Get Intro())
   > n13 = Branch(Condition=Get bWillReceiveBonus() -> (bWillReceiveBonus:bool))
   > SetActiveWidgetIndex(self=Get Switcher() -> (Switcher:Widget Switcher Object Reference), Index=0)
   > n14 = Cast To FreshCheeseDisplayItem(Object=Get DisplayItem_1() -> (DisplayItem_1:Fort Multiple Received Item Widget Object Reference))
   > ShowReward(self=@n14)
   > RarityVisuals(Rarity=Get DevRarity(self=@n14) -> (DevRarity=Common))
   > Set Timer by Event(Delegate=@n7, Time=0.5, bLooping=false)
   > Set Timer by Event(Delegate=@n9, Time=1, bLooping=false)
   > Set Timer by Event(Delegate=@n1, Time=2, bLooping=false)
   ShowNoFriendsMessage(execute=@n13.else)
   > Set Timer by Event(Delegate=@n11, Time=6, bLooping=false)
   Event StartNotificationOutro() {is=title K2Node_Event}
   n15 = Get Outro() -> (Outro:Widget Animation Object Reference)
   > PlayAnimation(InAnimation=@n15)
   > Delay(Duration=GetEndTime(self=@n15))
   > FinishNotificationOutro(self:Fort Notification Entry Object Reference)
  ]}
 ShowNoFriendsMessage {nodes [
   ShowNoFriendsMessage() {is=title K2Node_FunctionEntry}
   > SetActiveWidgetIndex(self=Get Switcher(), Index=1)
   > RarityVisuals(Rarity=Common)
  ]}
]
variables [
 DevRarity {type=EFortRarity, default=Common}
 IntroSpeed {type=number, default=3}
]
widgets {
 info {
  inheritedWidgetCount=4
  parentClass=/Script/FortniteUI.FortNotificationEntry_MultipleReceivedItem
  rootWidgetClass=&Overlay
  widgetCount=18
 }
 widgets [
  Overlay_5 {widgetClassPath=&Overlay, widget=Overlay_5}
  SizeBox_4 {
   widgetClassPath=&SizeBox
   parent=Overlay_5
   widget=SizeBox_4
   properties {bOverride_MinDesiredHeight=true, bOverride_MinDesiredWidth=true, minDesiredHeight=104, minDesiredWidth=405, widthOverride=448}
   slot {class=OverlaySlot, properties {horizontalAlignment=HAlign_Fill}}
  }
  Backing {
   widgetClassPath=&CommonBorder
   bIsVariable=true
   parent=SizeBox_4
   widget=Backing
   properties {
    background {
     bIsDynamicallyLoaded=false
     drawAs=Box
     imageType=NoImage
     mirroring=NoMirror
     resourceName=None
     resourceObject=/Game/Athena/HUD/Quests/Art/M_UI_NPC_Freebie_Container
     tiling=NoTile
     imageSize=<48 48>
     margin=<ltrb 0.5 0.5 0.5 0.5>
     outlineSettings {
      bUseBrushTransparency=false
      roundingType=HalfHeightRadius
      width=0
      color {colorUseRule=UseColor_Specified, specifiedColor=<rgba 0 0 0 0>}
      cornerRadii=<0 0 0 0>
     }
     tintColor {colorUseRule=UseColor_Specified, specifiedColor=<rgba 1 1 1 1>}
     uVRegion {bIsValid=false, max=<0 0>, min=<0 0>}
    }
    padding=<ltrb 0 0 0 0>
   }
   slot {class=SizeBoxSlot}
  }
  Switcher {
   widgetClassPath=&WidgetSwitcher
   bIsVariable=true
   parent=Backing
   widget=Switcher
   slot {class=BorderSlot, properties {padding=<ltrb 0 0 0 0>}}
  }
  Rewards {widgetClassPath=&Overlay, parent=Switcher, widget=Rewards, slot {class=WidgetSwitcherSlot}}
  RewardFlare {
   widgetClassPath=&Image
   bIsVariable=true
   parent=Rewards
   widget=RewardFlare
   properties {
    brush {
     bIsDynamicallyLoaded=false
     drawAs=Image
     imageType=NoImage
     mirroring=NoMirror
     resourceName=None
     resourceObject=/Game/UI/Foundation/Onboarding/Materials/M_FX_TriangleScroll_Toast
     tiling=NoTile
     imageSize=<64 64>
     margin=<ltrb 0 0 0 0>
     outlineSettings {
      bUseBrushTransparency=false
      roundingType=HalfHeightRadius
      width=0
      color {colorUseRule=UseColor_Specified, specifiedColor=<rgba 0 0 0 0>}
      cornerRadii=<0 0 0 0>
     }
     tintColor {colorUseRule=UseColor_Specified, specifiedColor=<rgba 1 1 1 1>}
     uVRegion {bIsValid=false, max=<0 0>, min=<0 0>}
    }
    renderTransform {angle=0, scale=<-1 1>, shear=<0 0>, translation=<0 0>}
   }
   slot {class=OverlaySlot, properties {horizontalAlignment=HAlign_Fill, verticalAlignment=VAlign_Fill, padding=<ltrb 0 0 2 0>}}
  }
  VerticalBox_93 {
   widgetClassPath=&VerticalBox
   parent=Rewards
   widget=VerticalBox_93
   slot {class=OverlaySlot, properties {verticalAlignment=VAlign_Center, padding=<ltrb 32 0 0 0>}}
  }
  SizeBox_1 {
   widgetClassPath=&SizeBox
   parent=VerticalBox_93
   widget=SizeBox_1
   properties {bOverride_MaxDesiredWidth=true, heightOverride=48.5222, maxDesiredWidth=408, widthOverride=408}
   slot {class=VerticalBoxSlot, properties {horizontalAlignment=HAlign_Left, verticalAlignment=VAlign_Center, padding=<ltrb 0 2 0 -5>}}
  }
  ScaleBox_1 {
   widgetClassPath=&ScaleBox
   parent=SizeBox_1
   widget=ScaleBox_1
   properties {stretch=ScaleToFitX, stretchDirection=DownOnly}
   slot {class=SizeBoxSlot}
  }
  Text_Header {
   widgetClassPath=&CommonTextBlock
   bInherited=true
   bIsVariable=true
   parent=ScaleBox_1
   widget=Text_Header
   properties {
    scrollStyle=&TextScrollStyle-Base_C
    text="PERSONAL BONUS!"
    textTransformPolicy=ToUpper
    font {
     bForceMonospaced=false
     bMaterialIsStencil=false
     fontMaterial=null
     fontObject=&Burbank
     letterSpacing=0
     monospacedWidth=1
     size=32
     skewAmount=0
     typefaceFontName=BurbankRegular-Black
     outlineSettings {
      bApplyOutlineToDropShadows=false
      bMiteredCorners=false
      bSeparateFillAlpha=false
      outlineMaterial=null
      outlineSize=0
      outlineColor=<rgba 0.2882 0.4028 0.651 1>
     }
    }
    shadowOffset=<6 3>
   }
   slot {class=ScaleBoxSlot, properties {horizontalAlignment=HAlign_Left}}
  }
  HorizontalBox_0 {
   widgetClassPath=&HorizontalBox
   parent=VerticalBox_93
   widget=HorizontalBox_0
   slot {class=VerticalBoxSlot, properties {padding=<ltrb 0 0 0 12>}}
  }
  DisplayItem_1 {
   widgetClassPath=&/FreshCheeseDisplayItem
   bInherited=true
   bIsVariable=true
   parent=HorizontalBox_0
   widget=DisplayItem_1
   slot {class=HorizontalBoxSlot, properties {padding=<ltrb 0 0 20 0>}}
  }
  DisplayItem_2 {
   widgetClassPath=&FreshCheeseDisplayItem
   bInherited=true
   bIsVariable=true
   parent=HorizontalBox_0
   widget=DisplayItem_2
   slot {class=HorizontalBoxSlot, properties {padding=<ltrb 0 0 20 0>}}
  }
  DisplayItem_3 {
   widgetClassPath=&FreshCheeseDisplayItem
   bInherited=true
   bIsVariable=true
   parent=HorizontalBox_0
   widget=DisplayItem_3
   slot {class=HorizontalBoxSlot}
  }
  Message {widgetClassPath=&Overlay, parent=Switcher, widget=Message, slot {class=WidgetSwitcherSlot}}
  VerticalBox_TextStack {
   widgetClassPath=&VerticalBox
   parent=Message
   widget=VerticalBox_TextStack
   slot {class=OverlaySlot, properties {verticalAlignment=VAlign_Center, padding=<ltrb 32 32 32 32>}}
  }
  RichText_SecondaryMessage {
   widgetClassPath=/Script/UMG.RichTextBlock
   bIsVariable=true
   parent=VerticalBox_TextStack
   widget=RichText_SecondaryMessage
   properties {
    text="<Secondary>REMINDER:</>"
    textStyleSet=/Game/UI/InGame/EventNotifications/MultipleReceivedItem_TextStyles
    renderTransform {angle=0, scale=<1 1>, shear=<-8 0>, translation=<0 0>}
   }
   slot {class=VerticalBoxSlot, properties {horizontalAlignment=HAlign_Left}}
  }
  RichText_PrimaryMessage {
   widgetClassPath=&RichTextBlock
   bIsVariable=true
   parent=VerticalBox_TextStack
   widget=RichText_PrimaryMessage
   properties {
    text="<levelNo>Squad up with</> <levelNoCyan>friends</> <levelNo>for extra</><levelNoCyan> bonus items</>"
    textStyleSet=&MultipleReceivedItem_TextStyles
    renderTransform {angle=0, scale=<1 1>, shear=<-8 0>, translation=<0 0>}
   }
   slot {class=VerticalBoxSlot, properties {horizontalAlignment=HAlign_Left}}
  }
 ]
}
