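# Sound cue for the Athena punchcard quest pop-up, routed to the non-critical in-game UI sound class.
# A single wave player for the Season 13 pop-up sound feeds an Enveloper whose volume curve fades in from near silence over a quarter second.
# A concurrency override allows only one instance, stopping the oldest when a new one plays.
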SoundCue /Game/Athena/Sounds/Quests/Punchcard/Punchcard_PopUp_Cue
root=Enveloper_0
nodes [
 {
  id=Enveloper_0
  children [WavePlayer_0]
  properties {
   pitchCurve {editorCurveData {keys [{time=0}]}}
   volumeCurve {editorCurveData {keys [
      {interpMode=RCIM_Cubic, tangentMode=RCTM_Break, time=0, value=0.01}
      {arriveTangent=2.2266, interpMode=RCIM_Cubic, leaveTangent=2.2266, tangentMode=RCTM_User, time=0.25, value=1}
     ]}}
  }
 }
 {id=WavePlayer_0, properties {soundWaveAssetPtr=&/S13_Punchcard_PopUp}}
]
properties {
 bOverrideConcurrency=true
 maxDistance=2097152
 soundClassObject=&Fort_UI_InGame_NonCritical
 volumeMultiplier=0.6
 concurrencyOverrides {maxCount=1, resolutionRule=StopOldest}
}
