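# In-game punch card notification widget, based on FortNotificationEntry_PunchCard, showing a challenge punch card and a collapsed XP sub-notification.
# Intro fades, slides and scales the stack in, then bounces the card; its start binding triggers a pop-up sound and suppressing sound mix.
# FinishNotificationIntro fires right after playback begins, while FinishNotificationOutro waits for the fade-out to end.
# The mix pops on outro start, outro finish and Destruct; debug prints remain.
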WidgetBlueprint /FortniteCoreUI/UI/InGame/EventNotifications/PunchCardNotification
animations [
 Intro {
  binding_count=2
  length_seconds=1.1333
  object=&self:Intro
  sequence {
   bindings [
    PunchCard {
     id=#1
     possessed_class=ChallengePunchCard_C
     tracks [{
       type=MovieScene2DTransformTrack
       sections [{
         type=MovieScene2DTransformSection
         channels [
          Translation.X {type=MovieSceneScriptingFloatChannel}
          Translation.Y {type=MovieSceneScriptingFloatChannel}
          Angle {type=MovieSceneScriptingFloatChannel}
          Scale.X {type=MovieSceneScriptingFloatChannel, keys [[47,1],[56,0.95],[61,1.025],[68,1]]}
          Scale.Y {type=MovieSceneScriptingFloatChannel, keys [[47,1],[56,0.95],[61,1.025],[68,1]]}
          Shear.X {type=MovieSceneScriptingFloatChannel}
          Shear.Y {type=MovieSceneScriptingFloatChannel}
         ]
        }]
      }]
    }
    VerticalBox_0 {
     id=#2
     possessed_class=VerticalBox
     tracks [
      {
       type=MovieSceneFloatTrack
       display_name=Render Opacity
       sections [{type=MovieSceneFloatSection, channels [None {type=MovieSceneScriptingFloatChannel, keys [[0.001,0],[30.001,1]]}]}]
      }
      {
       type=MovieScene2DTransformTrack
       sections [{
         type=MovieScene2DTransformSection
         channels [
          Translation.X {type=MovieSceneScriptingFloatChannel, keys [[0.001,-100],[30.001,0]]}
          Translation.Y {type=MovieSceneScriptingFloatChannel}
          Angle {type=MovieSceneScriptingFloatChannel}
          Scale.X {type=MovieSceneScriptingFloatChannel, keys [[0.001,0.8],[30.001,1]]}
          Scale.Y {type=MovieSceneScriptingFloatChannel, keys [[0.001,0.8],[30.001,1]]}
          Shear.X {type=MovieSceneScriptingFloatChannel}
          Shear.Y {type=MovieSceneScriptingFloatChannel}
         ]
        }]
      }
     ]
    }
   ]
   display_rate {numerator=60}
  }
 }
 Outro {
  binding_count=1
  length_seconds=0.25
  object=&self:Outro
  sequence {bindings [VerticalBox_0 {
     id=#3
     possessed_class=VerticalBox
     tracks [{
       type=MovieSceneFloatTrack
       display_name=Render Opacity
       sections [{type=MovieSceneFloatSection, channels [None {type=MovieSceneScriptingFloatChannel, keys [[0,1],[5,0]]}]}]
      }]
    }]}
 }
]
graphs [
 EventGraph {nodes [
   Play Intro() {is=title K2Node_CustomEvent}
   > PlaySound2D(Sound=/Game/Athena/Sounds/Quests/Punchcard/Punchcard_PopUp_Cue)
   > PushSoundMixModifier!(execute, InSoundMixModifier:Sound Mix Object Reference=/Game/Athena/Sounds/Quests/Punchcard/Punchcard_Suppress_MixMod) -> (then)
   n1 = Event Destruct()
   PopSoundMixModifier!(execute=[@n2.then, @n1.then], InSoundMixModifier:Sound Mix Object Reference=&Punchcard_Suppress_MixMod) -> (then)
   Event Construct()
   > SetRenderOpacity(self=Get VerticalBox_0() -> (VerticalBox_0:Vertical Box Object Reference), InOpacity=0)
   > BindToAnimationStarted!(execute, self:User Widget Object Reference, Animation:Widget Animation Object Reference=Get Intro(), Delegate:Delegate=Create Event()) -> (then)
   Event StartNotificationIntro() {is=title K2Node_Event}
   > PlayAnimationForward(InAnimation=Get Intro())
   > FinishNotificationIntro(self:Fort Notification Entry Object Reference)
   > PrintString!(execute, InString="Played punchcard Intro anim !!", bPrintToScreen, bPrintToLog, TextColor=(R=1,G=0,B=0.5137,A=1), Duration=3, Key) -> (then)
   Event StartNotificationOutro() {is=title K2Node_Event}
   > n2 = PlayAnimation(InAnimation=Get Outro() -> (Outro:Widget Animation Object Reference))
   "Animation Finished (Outro)"!() -> (OutputDelegate:Delegate, then:Exec) {is=title K2Node_WidgetAnimationEvent}
   > FinishNotificationOutro(self:Fort Notification Entry Object Reference)
   > SetRenderOpacity(self=Get VerticalBox_0(), InOpacity=0)
   > PrintString(InString="Played punchcard outro anim !!", TextColor=(R=1,G=0,B=0.5137,A=1), Duration=3)
   > PopSoundMixModifier(InSoundMixModifier=&Punchcard_Suppress_MixMod)
  ]}
 Sequencer Events {nodes [Reset render opacity(self:Xp Sub Notification Object Reference)]}
]
variables []
widgets {
 info {
  inheritedWidgetCount=2
  parentClass=/Script/FortniteUI.FortNotificationEntry_PunchCard
  rootWidgetClass=&VerticalBox
  widgetCount=3
 }
 widgets [
  VerticalBox_0 {widgetClassPath=&VerticalBox, bIsVariable=true, widget=VerticalBox_0}
  PunchCard {
   widgetClassPath=/Game/Athena/UI/Challenges/ChallengePunchCard
   bInherited=true
   bIsVariable=true
   parent=VerticalBox_0
   widget=PunchCard
   properties {isIngame=true}
   slot {class=VerticalBoxSlot}
  }
  XpSubNotification {
   widgetClassPath=&/XpSubNotification
   bInherited=true
   bIsVariable=true
   parent=VerticalBox_0
   widget=XpSubNotification
   properties {visibility=Collapsed, renderTransformPivot=<0.5 0>}
   slot {class=VerticalBoxSlot, properties {verticalAlignment=VAlign_Top}}
  }
 ]
}
