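# In-game HUD widget for game event notifications, a Widget Blueprint subclass of the native FortGameEventNotifications class.
# Its layout holds a NotificationUpdatesBox vertical box, scaled down around a left-center pivot, and a TutorialHighlight widget.
# The event graph is empty and no variables are declared, so the Blueprint supplies only layout and leaves any behavior to the parent class.
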WidgetBlueprint /FortniteCoreUI/UI/InGame/EventNotifications/EventNotifications
graphs [EventGraph {links [], nodes []}]
variables []
widgets {
 info {inheritedWidgetCount=2, parentClass=/Script/FortniteUI.FortGameEventNotifications, rootWidgetClass=&VerticalBox, widgetCount=2}
 widgets [
  NotificationUpdatesBox {
   widgetClassPath=&VerticalBox
   bInherited=true
   widget=NotificationUpdatesBox
   properties {renderTransform {angle=0, scale=<0.8 0.8>, shear=<0 0>, translation=<0 0>}, renderTransformPivot=<0 0.5>}
  }
  TutorialHighlight {widgetClassPath=&TutorialHighlight}
 ]
}
