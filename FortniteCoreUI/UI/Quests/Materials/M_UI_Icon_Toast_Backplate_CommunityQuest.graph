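# Translucent UI material that draws the Community Quest toast icon: a two-color backplate with a glowing, tintable icon.
# The backplate comes from an MF_UI_Circle SDF on scaled, offset 9-slice UVs, with two anti-aliased thresholds separating outline from interior.
# A vertical offset attempts to center the shape using a magic-number divisor.
# The icon is decoded from an MSDF texture and combined with an exponential radial glow.
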Material /FortniteCoreUI/UI/Quests/Materials/M_UI_Icon_Toast_Backplate_CommunityQuest
graph {
 properties {
  bAutomaticallySetUsageInEditor=false
  bUsedWithStaticMesh=false
  blendMode=BLEND_Translucent
  materialDomain=MD_UI
  usedShadingModels=Default Lit
 }
 nodes [
  LinearInterpolate(A=NamedRerouteUsage(), Alpha=NamedRerouteUsage())
  Multiply(A=NamedRerouteUsage(), B=ScalarParameter() {defaultValue=1, parameterName=TopPercentHeight})
  n1 = MaterialFunctionCall(
   UVs=MaterialFunctionCall() {materialFunction=&GetUserInterfaceUV}.9-Slice UV
   Texture Scale=Constant() {r=0.85}
  ) {materialFunction=&ScaleUVsByCenter}
  n2 = Constant() {r=0.5}
  n3 = ScalarParameter() {defaultValue=1, parameterName=TopPercentHeight}
  n4 = MaterialFunctionCall(X=Constant(), Y=Divide(A=Subtract(A=@n2, B=@n3)) {constB=2.25}) {
   desc="This is an attempt to center\r\nthe shape by offsetting by the\r\npercent difference. 2.25 is a\r\nmagic number"
   materialFunction=&MakeFloat2
  }
  n5 = Constant()
  n6 = Divide(A=NamedRerouteUsage()) {constB=4}
  n7 = MaterialFunctionCall(
   UVs=Add(
    A=Add(A=@n1.UVs, B=@n4.Result) {constB=0}
    B=MaterialFunctionCall(X=@n5, Y=@n6) {materialFunction=&MakeFloat2}.Result
   )
   Size=NamedRerouteUsage()
  ) {materialFunction=&MF_UI_Circle}
  n8 = MaterialFunctionCall(DistanceField=@n7.SDF, Threshold=Constant() {r=0.05}) {materialFunction=&MF_UI_AntiAliasing}
  n9 = Constant() {r=0.15}
  NamedRerouteDeclaration(None=Max(
    A=NamedRerouteDeclaration(None=@n8.Inner) {name=InteriorShape, nodeColor=<rgba 0 0 0 1>}
    B=MaterialFunctionCall(DistanceField=@n7.SDF, Threshold=@n9) {materialFunction=&MF_UI_AntiAliasing}.Inner
   )) {name=Shape, nodeColor=<rgba 0 0 0 1>}
  n10 = VectorParameter() {parameterName=OutlineColor, defaultValue=<rgba 0.2195 0.0144 0.6584 1>}
  n11 = VectorParameter() {parameterName=InnerColor, defaultValue=<rgba 0.4452 0.0409 0.807 1>}
  n12 = NamedRerouteUsage()
  n13 = Add(A=NamedRerouteUsage(), B=Constant())
  NamedRerouteDeclaration(None=Multiply(A=LinearInterpolate(A=@n10.RGB, B=@n11.RGB, Alpha=@n12), B=@n13)) {name=BackplateColor, nodeColor=<rgba 0 0 0 1>}
  NamedRerouteDeclaration(None=@n13) {name=BackplateMask, nodeColor=<rgba 0 0 0 1>}
  NamedRerouteDeclaration(None=MaterialFunctionCall(
    UVs=TextureCoordinate()
    Radius=ScalarParameter() {defaultValue=0.5, parameterName=GlowScale}
    Density=ScalarParameter() {defaultValue=1.2, parameterName=GlowDensity}
   ) {materialFunction=&RadialGradientExponential}.RadialGradientExponential) {name=Glow, nodeColor=<rgba 1 0 0.4688 1>}
  n14 = TextureSample(UVs=MaterialFunctionCall(UVs=TextureCoordinate(), Texture Scale=ScalarParameter() {defaultValue=0.7, parameterName=TextureScale}) {materialFunction=&ScaleUVsByCenter}.UVs) {
   samplerSource=SSM_Clamp_WorldGroupSettings
   samplerType=SAMPLERTYPE_LinearColor
   texture=/FortUILibrary/Textures/Icons/T_UI_Icon_UIKit_CommunityQuest_MSDF
  }
  NamedRerouteDeclaration(None=MaterialFunctionCall(
    DistanceField=MaterialFunctionCall!(MSDF RGB=@n14.RGB) {materialFunction=/FortUILibrary/MaterialFunctions/Shapes/Modifiers/MF_UI_MSDFtoSDF}.SDF
    Threshold=ScalarParameter() {defaultValue=0.52, parameterName=Threshold}
   ) {materialFunction=&MF_UI_AntiAliasing}.Outer) {name=Icon Texture, nodeColor=<rgba 0 0 0 1>}
  n15 = NamedRerouteUsage()
  n16 = VectorParameter() {parameterName=Glow, defaultValue=<rgba 0.1699 0.9473 0.8251 1>}
  n17 = NamedRerouteUsage()
  n18 = VectorParameter() {parameterName=Tint, defaultValue=<rgba 0.6332 0.6869 0.9473 1>}
  n19 = ScalarParameter() {parameterName=TintIntensity}
  NamedRerouteDeclaration(None=LinearInterpolate(A=Multiply(A=@n15, B=@n16.RGB), B=LinearInterpolate(A=@n17, B=@n18.RGB, Alpha=@n19), Alpha=NamedRerouteUsage())) {name=Icon with Glow, nodeColor=<rgba 0 0 0 1>}
  NamedRerouteDeclaration(None=ScalarParameter() {defaultValue=0.8, parameterName=Height}) {name=Height, nodeColor=<rgba 0 0 0 1>}
  NamedRerouteDeclaration(None=ScalarParameter() {defaultValue=0.9, parameterName=Width}) {name=Width, nodeColor=<rgba 0 0 0 1>}
  NamedRerouteUsage()
  ScalarParameter() {parameterName=TopWidth}
 ]
}
