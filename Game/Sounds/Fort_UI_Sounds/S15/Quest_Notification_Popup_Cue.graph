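# Sound cue for a quest notification popup in the UI, routed to the Fort_UI sound class.
# A single wave player reuses the S13 mini-challenge notification popup wave, pitched up and played at low volume.
# A concurrency override allows one instance at a time, stopping the oldest when a new one starts.
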SoundCue /Game/Sounds/Fort_UI_Sounds/S15/Quest_Notification_Popup_Cue
root=WavePlayer_0
nodes [{id=WavePlayer_0, properties {soundWaveAssetPtr=/Game/Athena/Sounds/Quests/MiniChallenges/S13_Notification_Popup}}]
properties {
 bOverrideConcurrency=true
 maxDistance=2097152
 pitchMultiplier=1.5
 soundClassObject=&Fort_UI
 volumeMultiplier=0.2
 concurrencyOverrides {maxCount=1, resolutionRule=StopOldest}
}
